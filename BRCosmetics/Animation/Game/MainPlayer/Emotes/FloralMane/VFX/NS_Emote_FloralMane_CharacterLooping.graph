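# Looping character effect for the FloralMane emote that plays water-style ripples and splashes at the character's feet.
# A change in User.BurstIndex fires a burst in both emitters and picks the foot bone or socket to spawn from.
# FootRipples layers mesh and sprite rings, while FootSplashes_Condensed spawns two spawn groups of time-of-day-tinted, gravity-affected sprites.
# Both emitters switch particles between world and local space using an owner-velocity threshold.
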NiagaraSystem /BRCosmetics/Animation/Game/MainPlayer/Emotes/FloralMane/VFX/NS_Emote_FloralMane_CharacterLooping
dependencies {
 usedDataInterfaces [
  &NiagaraDataInterfaceCurve
  &NiagaraDataInterfaceSpriteRendererInfo
  &NiagaraDataInterfaceSkeletalMesh
 ]
 usedDynamicInputs []
 usedModules [
  &self:SystemSpawnScript.SystemScriptSource.SystemScriptGraph.NiagaraNodeAssignment_0.SetVariables_#1
  &SystemState
  &BlendSpaceHelper
 ]
}
emitters {
 FootRipples {
  emitterUpdateScript [
   &EmitterState()
   /Game/Effects/Niagara/Modules/Logic/UpdateValueStateChange(InputIntParam:NiagaraInt32=&MakeIntFromFloat(Float=$User.BurstIndex, Conversion Mode=::NewEnumerator1))
   &self:TriggerSpawn(
    Spawn Count:NiagaraInt32=&MergeHelper_IntSwitch(
     0=1
     1=1
     2=1
     3=1
     4=1
     5=1
     6=1
     7=1
     8=1
     Index=&MakeIntFromFloat(Float=$User.BurstIndex, Conversion Mode=::NewEnumerator1)
     Num=::NewEnumerator7
    )
    Spawn Group:NiagaraInt32=&MakeIntFromFloat(Float=0, Conversion Mode=::NewEnumerator1)
    Trigger:NiagaraBool=$Emitter.UpdateValueStateChange.Changed
   ) {module=TriggerSpawn}
  ]
  particleSpawnScript [
   &InitializeParticle(
    Color=<rgba 2.5115 4.0692 5 1>
    Lifetime Max=0.35
    Lifetime Min=0.22
    Lifetime Mode=::NewEnumerator1
    Mesh Scale Mode=::NewEnumerator3
    Mesh Uniform Scale=0.026
    Sprite Rotation Mode=::NewEnumerator2
    Sprite Size Mode=::NewEnumerator1
    Uniform Sprite Size Max=55
    Uniform Sprite Size Min=20
    UsePositionOffset=false
    Write Material Random=true
   )
   &InitialMeshOrientation(
    Mesh Axis To Orient=<0 1 0>
    Mesh Orientation Mode=::NewEnumerator2
    Rotation=&RandomRangeVector(Minimum=<0 -1 0>, Maximum=<0 1 0>)
    Use Rotation Vector=true
   )
   &SkeletalMeshLocation(
    Bone / Socket Index=&MakeIntFromFloat(Float=$User.BurstIndex, Conversion Mode=::NewEnumerator1)
    MaskSpawnGroup=false
    Mesh Sampling Type=::NewEnumerator5
    Sampled Position Offset=<-2.5 10 0>
    Skeletal Mesh=$System.Foot SKM
    Socket Sampling Mode=::NewEnumerator5
   )
   /SetVariables(Particles.SpriteFacing=<0 0 1>) {isSetParameters=true}
  ]
  particleUpdateScript [
   &ParticleState()
   &ScaleMeshSize(Scale Factor=&VectorFromFloat(Value=&FloatFromCurve(FloatCurve=DataInterface() {propertyValues {Curve {
         defaultValue=1
         keys [
          {arriveTangent=1.3065, interpMode=RCIM_Cubic, leaveTangent=1.3065, tangentMode=RCTM_User, time=0.0027, value=0.233}
          {arriveTangent=1.0893, interpMode=RCIM_Cubic, leaveTangent=1.0893, tangentMode=RCTM_User, time=0.285, value=0.7005}
          {arriveTangent=0.0224, interpMode=RCIM_Cubic, leaveTangent=0.0224, tangentMode=RCTM_User, time=1, value=1}
         ]
        }}})))
   &FN_Radial_Material(Base Sprite Tile U=3, Depth Fade=1, Radial Mask Power=3, Sprite Extrusion Speed=5, Sprite Spin Speed=10)
   &BlendSpace("Position Blend (World <-> Local)":NiagaraFloat=&MakeCustomFloatFromBool(
     Bool=&SetBoolByFloatComparison(A=&VectorLength(Vector=$User.OwnerLocalVelocity), B=387)
     True Float=1
     False Float=0
     Translation Contribution:NiagaraFloat=1
     Rotation Contribution:NiagaraFloat=1
    ))
   &DynamicMaterialParameters(
    Index 0 Param 1=1
    Index 0 Param 2=1
    Index 0 Param 3=0.05
    Index 0 Param 4=1
    Module.Index0Float0Write=true
    Module.Index0Float1Write=true
    Module.Index0Float2Write=true
    Module.Index0Float3Write=true
   )
   &ScaleColor(
    Scale Alpha=&FloatFromCurve(FloatCurve=DataInterface() {propertyValues {Curve {
        defaultValue=1
        keys [
         <at -0.0107 0 RCIM_Cubic>
         {interpMode=RCIM_Cubic, tangentMode=RCTM_User, time=0.0369, value=1}
         {interpMode=RCIM_Cubic, tangentMode=RCTM_User, time=1, value=0}
        ]
       }}})
    ScaleA=true
    ScaleRGB=true
   )
   /SetVariables(
    Particles.Second Ripple Color:LinearColor=&MakeLinearColorFromVectorAndFloat(
     "Vector (RGB)"=&MakeVectorFromLinearColorRGB(Color=<rgba 0.3384 0.5681 0.7448 0.7739>)
     "Float (Alpha)"=&FloatFromCurve(
      FloatCurve=DataInterface() {propertyValues {Curve {
         defaultValue=1
         keys [
          {arriveTangent=34.6857, interpMode=RCIM_Cubic, leaveTangent=34.6857, tangentMode=RCTM_User, time=-0.0031, value=0}
          {arriveTangent=0.8678, interpMode=RCIM_Cubic, leaveTangent=0.8678, tangentMode=RCTM_User, time=0.0356, value=1}
          {arriveTangent=-1.3471, interpMode=RCIM_Cubic, leaveTangent=-1.3471, tangentMode=RCTM_User, time=0.2721, value=0.4717}
          {interpMode=RCIM_Cubic, tangentMode=RCTM_User, time=1, value=0}
         ]
        }}}
      Scale Curve=0.7
     )
    )
    Particles.Second Ripple Size:Vector2f=&Vector2DFromFloat(Value=&Multiply_Float(
      A=&RandomRangeFloat(Minimum=25, Maximum=30)
      B=&FloatFromCurve(FloatCurve=DataInterface() {propertyValues {Curve {
          defaultValue=0
          keys [
           {arriveTangent=3.3998, interpMode=RCIM_Cubic, leaveTangent=3.3998, tangentMode=RCTM_User, time=0, value=0}
           {arriveTangent=1.1202, interpMode=RCIM_Cubic, leaveTangent=1.1202, time=0.2019, value=0.528}
           {interpMode=RCIM_Cubic, tangentMode=RCTM_User, time=0.8927, value=1}
          ]
         }}})
     ))
    Particles.Sprire Ripple Color:LinearColor=&MakeLinearColorFromVectorAndFloat(
     "Vector (RGB)"=&MakeVectorFromLinearColorRGB(Color=<rgba 0.5673 0.7571 1 0.7217>)
     "Float (Alpha)"=&FloatFromCurve(
      FloatCurve=DataInterface() {propertyValues {Curve {
         defaultValue=1
         keys [
          {arriveTangent=0.9206, interpMode=RCIM_Cubic, leaveTangent=0.9206, tangentMode=RCTM_User, time=-0.0031, value=0}
          {interpMode=RCIM_Cubic, tangentMode=RCTM_User, time=0.1909, value=0.9547}
          {arriveTangent=-1.4228, interpMode=RCIM_Cubic, leaveTangent=-1.4228, tangentMode=RCTM_User, time=0.5781, value=0.2889}
          {interpMode=RCIM_Cubic, tangentMode=RCTM_User, time=1, value=0}
         ]
        }}}
      Scale Curve=0.6
     )
    )
    Particles.Sprite Ripple Size:Vector2f=&Vector2DFromFloat(Value=&Multiply_Float(
      A=45
      B=&FloatFromCurve(FloatCurve=DataInterface() {propertyValues {Curve {keys [
           {arriveTangent=1.4867, interpMode=RCIM_Cubic, leaveTangent=1.4867, tangentMode=RCTM_User, time=0, value=0}
           {arriveTangent=0.501, interpMode=RCIM_Cubic, leaveTangent=0.501, tangentMode=RCTM_User, time=1, value=1}
          ]}}})
     ))
   ) {isSetParameters=true}
  ]
  rendererClasses [&NiagaraMeshRendererProperties, &NiagaraSpriteRendererProperties]
  renderers [
   {rendererClass=&NiagaraMeshRendererProperties, rendererIndex=0}
   {rendererClass=&NiagaraSpriteRendererProperties, rendererIndex=1}
   {rendererClass=&NiagaraSpriteRendererProperties, rendererIndex=2}
  ]
 }
 FootSplashes_Condensed {
  emitterUpdateScript [
   &EmitterState()
   &UpdateValueStateChange(InputIntParam:NiagaraInt32=&MakeIntFromFloat(Float=$User.BurstIndex, Conversion Mode=::NewEnumerator1))
   &self:TriggerSpawn(
    Spawn Count:NiagaraInt32=&MergeHelper_IntSwitch(
     0=17
     1=17
     2=17
     3=17
     4=17
     5=17
     6=17
     7=17
     Index=&MakeIntFromFloat(Float=$User.BurstIndex, Conversion Mode=::NewEnumerator1)
     Num=::NewEnumerator6
    )
    Spawn Group:NiagaraInt32=&MakeIntFromFloat(Float=0, Conversion Mode=::NewEnumerator1)
    Trigger:NiagaraBool=$Emitter.UpdateValueStateChange.Changed
   ) {module=TriggerSpawn}
   &self:TriggerSpawn(
    Spawn Count:NiagaraInt32=&MergeHelper_IntSwitch(
     0=20
     1=20
     2=20
     3=20
     4=20
     5=20
     6=20
     7=20
     8=20
     Index=&MakeIntFromFloat(Float=$User.BurstIndex, Conversion Mode=::NewEnumerator1)
     Num=::NewEnumerator7
    )
    Spawn Group:NiagaraInt32=&MakeIntFromFloat(Float=1, Conversion Mode=::NewEnumerator1)
    Trigger:NiagaraBool=$Emitter.UpdateValueStateChange.Changed
   ) {module=TriggerSpawn001}
  ]
  particleSpawnScript [
   /SetVariables(Particles.VisibilityTag=&SpawnGroup()) {isSetParameters=true}
   &InitializeParticle(
    Color=&FNTimeOfDayColor(
     Night Color=&RandomRangeLinearColor(Minimum=<rgba 0.3649 0.5828 0.9219 0.7043>, Maximum=<rgba 0.1406 0.3946 0.75 0.8696>)
     Morning Color=&RandomRangeLinearColor(Minimum=<rgba 0.4685 0.6486 0.8177 1>, Maximum=<rgba 0.4664 0.6969 0.9427 1>)
     Day Color=&RandomRangeLinearColor(Minimum=<rgba 0.8201 0.8807 0.9583 1>, Maximum=<rgba 0.5504 0.7124 0.9427 1>)
     Evening Color=&RandomRangeLinearColor(Minimum=<rgba 0.4685 0.6486 0.8177 1>, Maximum=<rgba 0.4664 0.6969 0.9427 1>)
    )
    Lifetime=&MergeHelper_FloatSwitch(
     0=&RandomRangeFloat(Minimum=0.2, Maximum=0.5)
     1=&RandomRangeFloat(Minimum=0.3, Maximum=0.9)
     2=?
     3=?
     Index=&SpawnGroup()
     Num=::NewEnumerator0
    )
    Sprite Rotation Angle=&RandomRangeFloat(Minimum=0, Maximum=0.2)
    Sprite Rotation Mode=::NewEnumerator0
    Sprite Size Mode=::NewEnumerator3
    Uniform Sprite Size=&MergeHelper_FloatSwitch(
     0=&RandomRangeFloat(Minimum=26, Maximum=70)
     1=&RandomRangeFloat(Minimum=20, Maximum=55)
     2=?
     3=?
     Index=&SpawnGroup()
     Num=::NewEnumerator0
    )
    UsePositionOffset=false
    Write Material Random=true
   )
   &SkeletalMeshLocation(
    Bone / Socket Index=&MakeIntFromFloat(Float=$User.BurstIndex, Conversion Mode=::NewEnumerator1)
    MaskSpawnGroup=false
    Mesh Sampling Type=::NewEnumerator5
    Sampled Position Offset=&MergeHelper_VectorSwitch(
     0=&MaskVector(Mask Source=0, Mask Parameter=?, Mask=0, Masked Vector=<-2 24 -6>, Passthrough Vector=<-2 5 -6>)
     1=&MaskVector(Mask Source=0, Mask Parameter=?, Mask=0, Masked Vector=<2 24 6>, Passthrough Vector=<2 5 -6>)
     2=&MaskVector(Mask Source=0, Mask Parameter=?, Mask=0, Masked Vector=<-2 24 -6>, Passthrough Vector=<-2 5 -6>)
     3=&MaskVector(Mask Source=0, Mask Parameter=?, Mask=0, Masked Vector=<2 24 6>, Passthrough Vector=<2 5 -6>)
     4=&MaskVector(Mask Source=0, Mask Parameter=?, Mask=0, Masked Vector=<-2 24 -6>, Passthrough Vector=<-2 5 -6>)
     5=&MaskVector(Mask Source=0, Mask Parameter=?, Mask=0, Masked Vector=<2 24 -6>, Passthrough Vector=<2 5 -6>)
     6=&MaskVector(Mask Source=0, Mask Parameter=?, Mask=0, Masked Vector=<-2 24 -6>, Passthrough Vector=<-2 5 -6>)
     7=&MaskVector(Mask Source=0, Mask Parameter=?, Mask=0, Masked Vector=<2 24 -6>, Passthrough Vector=<2 5 -6>)
     Index=&MakeIntFromFloat(Float=$User.BurstIndex, Conversion Mode=::NewEnumerator1)
     Num=::NewEnumerator6
    )
    Skeletal Mesh=$System.Foot SKM
    Socket Sampling Mode=::NewEnumerator5
   )
   &SubUVAnimation(
    End Frame Range Override=3
    Evaluation Type=::NewEnumerator0
    Output Alternate Renderer Binding=true
    "Sprite Renderer "=DataInterface() {propertyValues {SpriteRenderer=&self:FootSplashes_Condensed_0.NiagaraSpriteRendererProperties_1}}
    SubUV Animation Mode:ENiagara_SubUVLookupModeV2=::NewEnumerator2
    UseEndFrame=true
    UseStartFrame=true
   )
   &SubUVAnimation(
    End Frame Range Override=3
    Evaluation Type=::NewEnumerator0
    Output Alternate Renderer Binding=true
    "Sprite Renderer "=DataInterface() {propertyValues {SpriteRenderer=&self:FootSplashes_Condensed_0.NiagaraSpriteRendererProperties_0}}
    SubUV Animation Mode:ENiagara_SubUVLookupModeV2=::NewEnumerator2
    UseEndFrame=true
    UseStartFrame=true
   ) {module=SubUVAnimation001}
   &ShapeLocation(
    MaskSpawnGroup=false
    Non Uniform Scale=<1 2 1>
    Override Seed=false
    Sphere Radius=&MergeHelper_FloatSwitch(0=10, 1=6, 2=?, 3=?, Index=&SpawnGroup(), Num=::NewEnumerator0)
    UseOwnerScale=true
   )
   &AddVelocity(
    Cone Angle=&MergeHelper_FloatSwitch(0=70, 1=45, 2=?, 3=?, Index=&SpawnGroup(), Num=::NewEnumerator0)
    Cone Axis=&MergeHelper_VectorSwitch(0=<0 0 1>, 1=<0 -1 0>, 2=?, 3=?, Index=&SpawnGroup(), Num=::NewEnumerator0)
    Inner Cone Angle=20
    Module.UseConeFalloff=false
    Module.UseVelDistribution=false
    Override Seed=false
    Velocity Mode=::NewEnumerator2
    Velocity Speed=&RandomRangeFloat(Minimum=20, Maximum=100)
   )
   &AddVelocity(Velocity=&MergeHelper_VectorSwitch(0=<0 -20 30>, 1=<0 -30 70>, 2=?, 3=?, Index=&SpawnGroup(), Num=::NewEnumerator0)) {module=AddVelocity001}
   &ShapeLocation(
    Cone Length=25
    MaskSpawnGroup=true
    Offset=<0 0 -3>
    Offset Mode=::NewEnumerator0
    Override Seed=false
    Shape Primitive=::NewEnumerator5
    Spawn Group Mask=1
    UseOwnerScale=true
   ) {module=ShapeLocation001}
   &ApplyInitialForces()
   &CameraOffset(Camera Offset Amount=7)
  ]
  particleUpdateScript [
   &ParticleState()
   &ScaleSpriteSize(Uniform Curve Sprite Scale=DataInterface() {propertyValues {Curve {keys [
        {arriveTangent=4.1785, interpMode=RCIM_Cubic, leaveTangent=4.1785, tangentMode=RCTM_User, time=0, value=0.15}
        {arriveTangent=1.1179, interpMode=RCIM_Cubic, leaveTangent=1.1179, tangentMode=RCTM_User, time=0.2042, value=0.6228}
        {arriveTangent=0.0224, interpMode=RCIM_Cubic, leaveTangent=0.0224, tangentMode=RCTM_User, time=1, value=1}
       ]}}})
   &GravityForce(Gravity=<0 0 -230>)
   &Drag(Drag=&RandomRangeFloat(Minimum=0.2, Maximum=1), Use Linear Drag=true, Use Rotational Drag=false)
   &SolveForcesAndVelocity(Clamp Velocity=false, Limit Acceleration=false)
   &ScaleColor(
    Scale Alpha=&FloatFromCurve(
     FloatCurve=DataInterface() {propertyValues {Curve {
        defaultValue=1
        keys [
         {interpMode=RCIM_Cubic, tangentMode=RCTM_User, time=0.1264, value=0}
         {interpMode=RCIM_Cubic, tangentMode=RCTM_User, time=0.4035, value=1}
         {arriveTangent=-0.0835, tangentMode=RCTM_User, time=0.9969, value=-0.0194}
        ]
       }}}
     Scale Curve=&RandomRangeFloat(Minimum=0.6, Maximum=0.85)
    )
    ScaleA=true
    ScaleRGB=true
   ) {module=ScaleColor001}
   /SetVariables(
    Particles.Splash 01 Dynam Param:Vector4f=&MakeVector4(
     X=&FloatFromCurve(FloatCurve=DataInterface() {propertyValues {Curve {keys [
          {arriveTangent=0.7033, interpMode=RCIM_Cubic, leaveTangent=0.7033, tangentMode=RCTM_User, time=0.1543, value=1}
          {arriveTangent=-1.4117, interpMode=RCIM_Cubic, leaveTangent=-1.4117, tangentMode=RCTM_User, time=0.5134, value=0.5054}
          {arriveTangent=0.3036, interpMode=RCIM_Cubic, leaveTangent=0.3036, tangentMode=RCTM_User, time=1, value=0}
         ]}}})
     Y=0
     Z=0
     W=0
    )
    Particles.Splash Dynam Param:Vector4f=&MakeVector4(
     X=&FloatFromCurve(FloatCurve=DataInterface() {propertyValues {Curve {keys [
          {arriveTangent=-0.1458, interpMode=RCIM_Cubic, leaveTangent=-0.1458, tangentMode=RCTM_User, time=0.1543, value=1}
          {arriveTangent=-1.4117, interpMode=RCIM_Cubic, leaveTangent=-1.4117, tangentMode=RCTM_User, time=0.6056, value=0.6113}
          {arriveTangent=0.3036, interpMode=RCIM_Cubic, leaveTangent=0.3036, tangentMode=RCTM_User, time=1, value=0}
         ]}}})
     Y=0
     Z=0
     W=0
    )
   ) {isSetParameters=true}
   &BlendSpace("Position Blend (World <-> Local)":NiagaraFloat=&MakeCustomFloatFromBool(
     Bool=&SetBoolByFloatComparison(A=&VectorLength(Vector=$User.OwnerLocalVelocity), B=387)
     True Float=1
     False Float=0
     Translation Contribution:NiagaraFloat=1
     Rotation Contribution:NiagaraFloat=1
    ))
  ]
  rendererClasses [&NiagaraSpriteRendererProperties]
  renderers [
   {rendererClass=&NiagaraSpriteRendererProperties}
   {rendererClass=&NiagaraSpriteRendererProperties, rendererIndex=1}
  ]
 }
}
summary {
 systemName=NS_Emote_FloralMane_CharacterLooping
 emitters [
  {emitterName=FootRipples, rendererClasses [&NiagaraMeshRendererProperties, &NiagaraSpriteRendererProperties]}
  {emitterName=FootSplashes_Condensed, rendererClasses [&NiagaraSpriteRendererProperties]}
 ]
 userVariables [
  User.BurstIndex {type=NiagaraFloat, defaultValue=0, description=""}
  User.OwnerLocalVelocity {type=Vector3f, description="", defaultValue {struct=Vector3f, value=<0 0 0>}}
 ]
}
