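# Looping water effects for the FloralMane emote's gadget prop, layering edge splashes, mesh spray, timed splash bursts, ribbon wake lines and a water trail.
# Emitters follow the prop through socket and skeletal mesh sampling, with GPU simulation for edge splashes and spray.
# User direction values suppress or rescale each side's spray and splashes as it turns.
# Wake ribbons bend with sine waves and road-bending values, and several colors follow time of day.
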NiagaraSystem /BRCosmetics/Animation/Game/MainPlayer/Emotes/FloralMane/VFX/NS_Emote_FloralMane_PropLooping
dependencies {
 usedDataInterfaces [
  &NiagaraDataInterfaceColorCurve
  &NiagaraDataInterfaceCurve
  &NiagaraDataInterfaceSkeletalMesh
  &NiagaraDataInterfaceSpriteRendererInfo
  &NiagaraDataInterfaceVector2DCurve
  &NiagaraDataInterfaceVectorCurve
  &NiagaraDataInterfaceSocketReader
 ]
 usedDynamicInputs []
 usedModules [
  &self:SystemSpawnScript.SystemScriptSource.SystemScriptGraph.NiagaraNodeAssignment_2.SetVariables_#1
  &SystemState
  &self:SystemSpawnScript.SystemScriptSource.SystemScriptGraph.NiagaraNodeAssignment_0.SetVariables_#2
  &self:SystemSpawnScript.SystemScriptSource.SystemScriptGraph.NiagaraNodeAssignment_1.SetVariables_#3
  &BlendSpaceHelper
 ]
}
emitters {
 Edge_Splashes {
  simTarget=GPUComputeSim
  emitterUpdateScript [
   &EmitterState()
   &SpawnRate(SpawnRate=320, Use Spawn Probability=false)
  ]
  particleSpawnScript [
   &PartitionParticles(Number Of Partitions=8)
   &InitializeParticle(
    Color Maximum=<rgba 2.3438 2.7094 3 1>
    Color Minimum=<rgba 1.3594 2.0399 3 1>
    Color Mode=::NewEnumerator2
    Lifetime Max=0.6
    Lifetime Min=0.25
    Lifetime Mode=::NewEnumerator1
    Sprite Size Max=<40 40>
    Sprite Size Min=<10 18>
    Sprite Size Mode=::NewEnumerator2
    UsePositionOffset=false
    Write Material Random=true
   )
   &SubUVAnimation(
    End Frame Range Override=3
    Evaluation Type=::NewEnumerator0
    "Sprite Renderer "=DataInterface() {propertyValues {SpriteRenderer=&self:Edge_Splashes_0.NiagaraSpriteRendererProperties_1}}
    SubUV Animation Mode:ENiagara_SubUVLookupModeV2=::NewEnumerator2
    UseEndFrame=true
    UseStartFrame=true
   )
  ]
  particleUpdateScript [
   /SetVariables(Particles.Size Random:NiagaraFloat=&Waveform(
     "Waveform [1]"=::NewEnumerator0
     "[1] Frequency"=&MergeHelper_FloatSwitch(
      0=&RandomRangeFloat(Minimum=0.05, Maximum=0.17)
      1=&RandomRangeFloat(Minimum=0.05, Maximum=0.2)
      2=&RandomRangeFloat(Minimum=0.05, Maximum=0.22)
      3=&RandomRangeFloat(Minimum=0.05, Maximum=0.18)
      4=&RandomRangeFloat(Minimum=0.05, Maximum=0.25)
      5=&RandomRangeFloat(Minimum=0.05, Maximum=0.2)
      6=&RandomRangeFloat(Minimum=0.25, Maximum=-0.84)
      7=&RandomRangeFloat(Minimum=0.25, Maximum=0.23)
      Index=$Particles.PartitionParticles.Partition
      Num=::NewEnumerator6
     )
     "[1] Amplitude Min/Max"=<0 1>
    )) {isSetParameters=true}
   &SolveForcesAndVelocity(Clamp Velocity=false, Limit Acceleration=false)
   &SkeletalMeshLocation(
    MaskSpawnGroup=false
    Mesh Sampling Type=::NewEnumerator5
    Orientation Sampling=::NewEnumerator1
    Sampled Position Offset=&MergeHelper_VectorSwitch(
     0=&RandomRangeVector(Minimum=<8 48 0>, Maximum=<12 25 0>)
     1=&RandomRangeVector(Minimum=<-8 48 0>, Maximum=<-12 25 0>)
     2=&RandomRangeVector(Minimum=<12 5 0>, Maximum=<16 25 0>)
     3=&RandomRangeVector(Minimum=<-12 5 0>, Maximum=<-16 25 0>)
     4=&RandomRangeVector(Minimum=<20 20 0>, Maximum=<25 -25 0>)
     5=&RandomRangeVector(Minimum=<-20 20 0>, Maximum=<-25 -25 0>)
     6=&RandomRangeVector(Minimum=<12 -28 0>, Maximum=<17 -45 0>)
     7=&RandomRangeVector(Minimum=<-12 -28 0>, Maximum=<-17 -45 0>)
     Index=$Particles.PartitionParticles.Partition
     Num=::NewEnumerator6
    )
    Skeletal Mesh=DataInterface() {propertyValues {
      DefaultMesh=null
      ExcludeBoneName=root
      OptionalComponentClass=null
      PreviewMesh=/BRCosmetics/Animation/Props/Gadgets_FloralMane/Meshes/FloralMane_Gadget
      SkinningMode=SkinOnTheFly
      SoftSourceActor=null
      SourceMode=Default
      UvSetIndex=0
      WholeMeshLOD=-1
      bExcludeBone=false
      bReadDeformedGeometry=true
      bRequireCurrentFrameData=true
      ComponentTags []
      FilteredBones []
      FilteredSockets [JNT_VFXSocket]
      MeshUserParameter {parameter {name=None, typeDefHandle {registeredTypeIndex=1}, varData []}}
      SamplingRegions []
     }}
    Socket Sampling Mode=::NewEnumerator5
   ) {module=SkeletalMeshLocation001}
   &FN_Sprite_Attributes(
    Alignment:Vector3f=<1 0 0>
    Create Alignment Attribute=true
    Create Facing Attribute=true
    Create Sprite Rotation Attribute=true
    Facing:Vector3f=&MergeHelper_VectorSwitch(
     0=<-1 1 1>
     1=<1 1 1>
     2=<-0.5 0.5 0.5>
     3=<0.5 0.5 0.5>
     4=<-0.5 1 1>
     5=<0.5 1 1>
     6=<-0.5 0.5 1>
     7=<0.5 0.5 1>
     Index=$Particles.PartitionParticles.Partition
     Num=::NewEnumerator6
    )
    Write to Module Namespace=true
   )
   &ParticleState()
   &ScaleSpriteSize(
    Scale Sprite Size Mode=::NewEnumerator0
    Uniform Scale Factor=&Multiply_Float(
     A=&FloatFromCurve(FloatCurve=DataInterface() {propertyValues {Curve {
         defaultValue=1
         keys [
          {arriveTangent=4.1785, interpMode=RCIM_Cubic, leaveTangent=4.1785, tangentMode=RCTM_User, time=0, value=0}
          {arriveTangent=1.1179, interpMode=RCIM_Cubic, leaveTangent=1.1179, tangentMode=RCTM_User, time=0.2042, value=0.6228}
          {arriveTangent=0.0224, interpMode=RCIM_Cubic, leaveTangent=0.0224, tangentMode=RCTM_User, time=1, value=1}
         ]
        }}})
     B=&MakeCustomFloatFromBool(
      Bool=&SetBoolByFloatComparison(A=$User.Direction_01, B=0, Comparison Type=::NewEnumerator6)
      True Float=&MergeHelper_FloatSwitch(0=0, 1=1, 2=0, 3=1, 4=0, 5=1, 6=0, 7=1, Index=$Particles.PartitionParticles.Partition, Num=::NewEnumerator6)
      False Float=&Multiply_Float(A=1, B=$Particles.Size Random)
     )
    )
   )
   &ScaleColor(
    Scale Alpha=&FloatFromCurve(
     FloatCurve=DataInterface() {propertyValues {Curve {keys [
         {arriveTangent=-0.1159, interpMode=RCIM_Cubic, leaveTangent=-0.1159, tangentMode=RCTM_User, time=0, value=1}
         {arriveTangent=-1.523, interpMode=RCIM_Cubic, leaveTangent=-1.523, tangentMode=RCTM_User, time=0.199, value=0.5089}
         <at 0.9633 0.0344 RCIM_Cubic>
        ]}}}
     Scale Curve=&FNTimeOfDayScalar(Night Scale=0.6, Morning Scale=0.85, Day Scale=1.05, Evening Scale=0.85)
    )
    Scale RGB=&MakeVectorFromLinearColorRGB(Color=&ColorFromCurve(ColorCurve=DataInterface() {propertyValues {
        AlphaCurve {keys []}
        BlueCurve {keys [<at 0.0194 1.25>, {arriveTangent=-0.2502, interpMode=RCIM_Cubic, time=0.1156, value=1}, <at 1 1 RCIM_Cubic>]}
        GreenCurve {keys [<at 0.0194 1.25>, {arriveTangent=-0.2502, interpMode=RCIM_Cubic, time=0.1156, value=1}, <at 1 1 RCIM_Cubic>]}
        RedCurve {keys [<at 0.0194 1.25>, {arriveTangent=-0.2502, interpMode=RCIM_Cubic, time=0.1156, value=1}, <at 1 1 RCIM_Cubic>]}
       }}))
    ScaleA=true
    ScaleRGB=true
   )
   &DynamicMaterialParameters(
    Index 0 Param 1=&FloatFromCurve(FloatCurve=DataInterface() {propertyValues {Curve {keys [
         <at 0.1543 1 RCIM_Cubic>
         {arriveTangent=-2.4268, interpMode=RCIM_Cubic, leaveTangent=-2.4268, tangentMode=RCTM_User, time=0.7835, value=0.2233}
         {arriveTangent=0.3036, interpMode=RCIM_Cubic, leaveTangent=0.3036, tangentMode=RCTM_User, time=1, value=0}
        ]}}})
    Module.Index0Float0Write=true
    Module.Index0Float1Write=true
    Module.Index0Float2Write=true
    Module.Index0Float3Write=true
   )
   &FN_Sprite_Attributes(
    Alignment:Vector3f=<1 0 0>
    Create Alignment Attribute=true
    Create Facing Attribute=true
    Facing:Vector3f=<0 0 1>
    Write to Module Namespace=true
   ) {module=FN_Sprite_Attributes001}
   /SetVariables(
    Particles.Linear Color=&MakeLinearColorFromVectorAndFloat(
     "Vector (RGB)"=&MakeVectorFromLinearColorRGB(Color=<rgba 0.2733 0.584 1 0.1304>)
     "Float (Alpha)"=&FloatFromCurve(
      FloatCurve=DataInterface() {propertyValues {Curve {
         defaultValue=1
         keys [
          {interpMode=RCIM_Cubic, tangentMode=RCTM_User, time=0, value=0}
          {arriveTangent=1.4478, interpMode=RCIM_Cubic, leaveTangent=1.4478, tangentMode=RCTM_User, time=0.1694, value=0.8588}
          {tangentMode=RCTM_User, time=1, value=0}
         ]
        }}}
      Scale Curve=0.08
     )
    )
    Particles.Linear Color001:LinearColor=&MakeLinearColorFromVectorAndFloat(
     "Vector (RGB)"=&MakeVectorFromLinearColorRGB(Color=<rgba 0.3599 0.691 1 0.5739>)
     "Float (Alpha)"=&FloatFromCurve(
      FloatCurve=DataInterface() {propertyValues {Curve {
         defaultValue=1
         keys [
          {interpMode=RCIM_Cubic, tangentMode=RCTM_User, time=0, value=0}
          {interpMode=RCIM_Cubic, tangentMode=RCTM_User, time=0.1632, value=1}
          {tangentMode=RCTM_User, time=1, value=0}
         ]
        }}}
      Scale Curve=&RandomRangeFloat(Minimum=0, Maximum=0.25)
     )
    )
    Particles.Move Pos:NiagaraPosition=&AddVectorToPosition(
     Position=$Particles.Position
     Vector=&MakeVector(
      X=&FloatFromCurve(
       FloatCurve=DataInterface() {propertyValues {Curve {defaultValue=1, keys [{tangentMode=RCTM_User, time=0, value=0}, {tangentMode=RCTM_User, time=1, value=1}]}}}
       Scale Curve=&MergeHelper_FloatSwitch(0=-10, 1=10, 2=-10, 3=10, 4=-10, 5=10, 6=-10, 7=10, Index=$Particles.PartitionParticles.Partition, Num=::NewEnumerator6)
      )
      Y=0
      Z=0
     )
    )
    Particles.Splash Pos:NiagaraPosition=&AddVectorToPosition(
     Position=$Particles.Position
     Vector=&MakeVector(
      X=0
      Y=&FloatFromCurve(
       FloatCurve=DataInterface() {propertyValues {Curve {
          defaultValue=0
          keys [<at 0 0 RCIM_Cubic>, {arriveTangent=-1.9544, interpMode=RCIM_Cubic, leaveTangent=-1.9544, tangentMode=RCTM_User, time=1, value=-1}]
         }}}
       Scale Curve=&MergeHelper_FloatSwitch(
        0=-30
        1=-30
        2=-30
        3=&RandomRangeFloat(Minimum=-30, Maximum=30)
        4=30
        5=30
        6=30
        7=30
        Index=$Particles.PartitionParticles.Partition
        Num=::NewEnumerator6
       )
      )
      Z=0
     )
    )
    Particles.Vector 2D:Vector2f=&Multiply_Vector2D_ByFloat(
     Float=&FloatFromCurve(
      FloatCurve=DataInterface() {propertyValues {Curve {
         defaultValue=1
         keys [
          {interpMode=RCIM_Cubic, tangentMode=RCTM_User, time=0, value=0}
          {interpMode=RCIM_Cubic, tangentMode=RCTM_User, time=0.1632, value=0.9418}
          {arriveTangent=-1.4982, interpMode=RCIM_Cubic, leaveTangent=-1.4982, tangentMode=RCTM_User, time=0.2504, value=0.905}
          {tangentMode=RCTM_User, time=1, value=0}
         ]
        }}}
      Scale Curve=4
     )
     Vector2D=$Particles.SpriteSize
    )
    Particles.Vector 2D001:Vector2f=&Multiply_Vector2D_ByFloat(
     Float=&Multiply_Float(
      A=&RandomRangeFloat(Minimum=0.25, Maximum=0.7)
      B=&FloatFromCurve(
       FloatCurve=DataInterface() {propertyValues {Curve {keys [<at 0.4 1 RCIM_Cubic>, {arriveTangent=3.4015, interpMode=RCIM_Cubic, leaveTangent=3.4015, tangentMode=RCTM_User, time=0.906, value=2}]}}}
       Scale Curve=1.2
      )
     )
     Vector2D=$Particles.SpriteSize
    )
   ) {isSetParameters=true}
  ]
  rendererClasses [&NiagaraSpriteRendererProperties]
  renderers [{rendererClass=&NiagaraSpriteRendererProperties}]
 }
 Edge_Spray_Mesh005 {
  simTarget=GPUComputeSim
  emitterSpawnScript [/SetVariables(Emitter.OverallEmissiveAdd:NiagaraFloat=&FN_FrontEndFloatModifier(Gameplay=0, Front End=0.5)) {isSetParameters=true}]
  emitterUpdateScript [
   &EmitterState(Inactive Response=::NewEnumerator1, Life Cycle Mode=::NewEnumerator1, Loop Behavior=::NewEnumerator1, Loop Duration=3600, UseLoopDelay=false)
   &SpawnRate(
    SpawnGroup:NiagaraInt32=&RandomRangeInt(Minimum=0, Maximum=1)
    SpawnRate=&Add_Float(
     A=230
     B=&MakeCustomFloatFromBool(
      Bool=&SetBoolByFloatComparison(A=$User.Direction_01, B=0, Comparison Type=::NewEnumerator5)
      True Float=70
      False Float=0
     )
    )
    Use Spawn Probability=false
   )
  ]
  particleSpawnScript [
   &FN_SetSpawnGroup(Number of Spawn Groups=2)
   /SetVariables(Particles.FrontToBack:NiagaraFloat=&RandomRangeFloat(Minimum=0, Maximum=1)) {isSetParameters=true}
   &InitializeParticle(
    Color=<rgba 0.6689 0.862 1 0.3739>
    Lifetime Max=1
    Lifetime Min=0.75
    Lifetime Mode=::NewEnumerator1
    Mesh Scale=&Multiply_Vector(
     A=&Multiply_VectorByFloat(Vector=&RandomRangeVector(Minimum=<0.325 0.8 0.325>, Maximum=<0.08 0.4 0.03>), Float=0.3)
     B=&VectorFromCurve(
      VectorCurve=DataInterface() {propertyValues {XCurve {keys [<at 0.0072 0.6344>,<at 0.9964 1.0913>]}, YCurve {keys [<at 0 1>,<at 1 1>]}, ZCurve {keys [<at 0 1>,<at 1 1>]}}}
      CurveIndex=$Particles.FrontToBack
     )
    )
    Mesh Scale Mode=::NewEnumerator4
    Sprite Rotation Mode=::NewEnumerator2
    UsePositionOffset=false
    Write Material Random=true
   ) {module=InitializeParticle001}
   &InitialMeshOrientation(
    Orientation Axis=<1 0 -0.5>
    Orientation Vector=&Multiply_Vector(
     A=&Add_Vector(A=&Lerp_Vector(A=<0.9 0 0>, B=<1 0 0>, Alpha=$Particles.FrontToBack), B=<1 -1 0>)
     B=&Multiply_Vector(
      A=&Lerp_Vector(A=<1 1 1>, B=<1 -1 1>, Alpha=&MakeFloatFromInt(Int=$Particles.SpawnGroup))
      B=&Lerp_Vector(A=<1 1 1>, B=<1 0.75 1>, Alpha=1)
     )
    )
    Rotation=&Multiply_Vector(
     A=&MergeHelper_VectorSwitch(0=<0 0 -0.2521>, 1=<0 0 0.2597>, 2=?, 3=?, Index=&SpawnGroup(), Num=::NewEnumerator0)
     B=&VectorFromCurve(
      VectorCurve=DataInterface() {propertyValues {XCurve {keys [<at 0 1>,<at 1 1>]}, YCurve {keys [<at 0 1>,<at 1 1>]}, ZCurve {keys [<at 0 0.6757>,<at 0.7459 0.761>,<at 1 0.849>]}}}
      CurveIndex=$Particles.FrontToBack
     )
    )
    Rotation Coordinate Space=::NewEnumerator0
    Use Orientation Vector=true
    Use Rotation Vector=true
   )
   /SetVariables(Particles.Scale=&Multiply_Vector(
     A=&VectorFromCurve(
      VectorCurve=DataInterface() {propertyValues {
        XCurve {keys [
          {arriveTangent=-0.0555, interpMode=RCIM_Cubic, leaveTangent=-0.0555, tangentMode=RCTM_User, time=0.009, value=0.9017}
          {arriveTangent=-0.1757, interpMode=RCIM_Cubic, leaveTangent=-0.1757, tangentMode=RCTM_User, time=1, value=1}
         ]}
        YCurve {keys [
          {arriveTangent=-0.0555, interpMode=RCIM_Cubic, leaveTangent=-0.0555, tangentMode=RCTM_User, time=0.009, value=0.9017}
          {arriveTangent=-0.1757, interpMode=RCIM_Cubic, leaveTangent=-0.1757, tangentMode=RCTM_User, time=1, value=1}
         ]}
        ZCurve {keys [
          {arriveTangent=-0.0555, interpMode=RCIM_Cubic, leaveTangent=-0.0555, tangentMode=RCTM_User, time=0.009, value=0.9017}
          {arriveTangent=-0.1757, interpMode=RCIM_Cubic, leaveTangent=-0.1757, tangentMode=RCTM_User, time=1, value=1}
         ]}
       }}
      CurveIndex=$Particles.FrontToBack
      Scale Curve=&Multiply_VectorByFloat(
       Vector=$Particles.Scale
       Float=&FloatFromCurve(
        FloatCurve=DataInterface() {propertyValues {Curve {
           defaultValue=1
           keys [
            <at 0.3946 -0.0118 RCIM_Cubic>
            {arriveTangent=0.3199, interpMode=RCIM_Cubic, leaveTangent=0.3199, tangentMode=RCTM_User, time=0.405, value=1}
           ]
          }}}
        CurveIndex=$Particles.FrontToBack
        Scale Curve=1.7
       )
      )
     )
     B=&Lerp_Vector(A=<0.6 0.6 0.6>, B=<0.8 0.8 0.8>, Alpha=$Particles.LeanLerp)
    )) {isSetParameters=true}
  ]
  particleUpdateScript [
   &ParticleState()
   &SocketLocation(
    SamplePosOffset=true
    Sampled Position Offset=&VectorFromCurve(
     VectorCurve=DataInterface() {propertyValues {
       XCurve {keys [
         {arriveTangent=-107.1167, time=-0.0001, value=9.6795}
         <at 0.1387 22.7439>
         <at 0.4255 -26.8096>
         <at 0.8777 -128.2796>
         <at 0.9955 -66.2887>
        ]}
       YCurve {keys [
         <at 0.2537 15 RCIM_Cubic>
         {arriveTangent=-92.7914, interpMode=RCIM_Cubic, leaveTangent=-92.7914, tangentMode=RCTM_User, time=1, value=-14.3043}
        ]}
       ZCurve {keys [<at 0 1>,<at 1 1>]}
      }}
     CurveIndex=$Particles.FrontToBack
     Scale Curve=&Lerp_Vector(A=<-0.15 3.5 1>, B=<0.15 3.5 1>, Alpha=&MakeFloatFromInt(Int=$Particles.SpawnGroup))
    )
    Socket Reader=$System.FloralMane Sockets
    Socket Sampling Mode=::NewEnumerator2
    VelocitySpeedLimit=false
   )
   &SocketLocation(
    Orientation Sampling=::NewEnumerator1
    Position Sampling=::NewEnumerator1
    SamplePosOffset=true
    Sampled Position Offset=<0 60 0>
    Sampled Position Offset Coordinate Space:ENiagaraSkelMeshTransforms=::NewEnumerator0
    Socket Reader=$System.FloralMane Sockets
    Socket Sampling Mode=::NewEnumerator2
    VelocitySpeedLimit=false
   ) {module=SocketLocation003}
   /SetVariables(Particles.Scale=&Multiply_Vector(
     A=$Particles.Initial.Scale
     B=&VectorFromCurve(
      VectorCurve=DataInterface() {propertyValues {
        XCurve {keys [
          {arriveTangent=4.6994, interpMode=RCIM_Cubic, leaveTangent=4.6994, tangentMode=RCTM_User, time=0, value=0.0795}
          {arriveTangent=0.4589, interpMode=RCIM_Cubic, leaveTangent=0.4589, tangentMode=RCTM_User, time=0.1447, value=0.7609}
          <at 1 1>
         ]}
        YCurve {keys [{arriveTangent=2.5596, interpMode=RCIM_Cubic, leaveTangent=2.5596, tangentMode=RCTM_User, time=0, value=0.0767}, <at 1 1>]}
        ZCurve {keys [
          <at 0 0 RCIM_Cubic>
          {arriveTangent=-2.3619, interpMode=RCIM_Cubic, leaveTangent=-2.3619, tangentMode=RCTM_User, time=0.336, value=1.5}
          <at 0.9988 0>
         ]}
       }}
      Scale Curve=&VectorFromFloat(Value=&MergeHelper_FloatSwitch(
        0=&MakeCustomFloatFromBool(
         Bool=&SetBoolByFloatComparison(A=$User.Direction_01, B=0)
         True Float=&Add_Float(A=1, B=$User.Direction_01)
         False Float=&MakeCustomFloatFromBool(
          Bool=&SetBoolByFloatComparison(A=$User.Direction_01, B=0, Comparison Type=::NewEnumerator6)
          True Float=&Add_Float(A=1, B=$User.Direction_01)
          False Float=1
         )
        )
        1=&MakeCustomFloatFromBool(
         Bool=&SetBoolByFloatComparison(A=$User.Direction_01, B=0, Comparison Type=::NewEnumerator6)
         True Float=&Add_Float(A=1, B=&Multiply_Float(A=$User.Direction_01, B=-1))
         False Float=&MakeCustomFloatFromBool(
          Bool=&SetBoolByFloatComparison(A=$User.Direction_01, B=0)
          True Float=&Subtract_Float(A=1, B=$User.Direction_01)
          False Float=1
         )
        )
        2=?
        3=?
        Index=&SpawnGroup()
        Num=::NewEnumerator0
       ))
     )
    )) {enabled=false, isSetParameters=true}
   /SetVariables(Particles.Scale=&Multiply_Vector(
     A=$Particles.Initial.Scale
     B=&VectorFromCurve(
      VectorCurve=DataInterface() {propertyValues {
        XCurve {keys [
          {arriveTangent=4.6994, interpMode=RCIM_Cubic, leaveTangent=4.6994, tangentMode=RCTM_User, time=0, value=0.0795}
          {arriveTangent=0.4589, interpMode=RCIM_Cubic, leaveTangent=0.4589, tangentMode=RCTM_User, time=0.1447, value=0.7609}
          <at 1 1>
         ]}
        YCurve {keys [{arriveTangent=2.5596, interpMode=RCIM_Cubic, leaveTangent=2.5596, tangentMode=RCTM_User, time=0, value=0.0767}, <at 1 1>]}
        ZCurve {keys [
          <at 0 0 RCIM_Cubic>
          {arriveTangent=-2.3619, interpMode=RCIM_Cubic, leaveTangent=-2.3619, tangentMode=RCTM_User, time=0.336, value=1.5}
          <at 0.9988 0>
         ]}
       }}
      Scale Curve=&VectorFromFloat(Value=&MergeHelper_FloatSwitch(
        0=&MakeCustomFloatFromBool(
         Bool=&SetBoolByFloatComparison(A=$User.Direction_01, B=0)
         True Float=&Add_Float(A=1, B=$User.Direction_01)
         False Float=&MakeCustomFloatFromBool(
          Bool=&SetBoolByFloatComparison(A=$User.Direction, B=0, Comparison Type=::NewEnumerator6)
          True Float=0
          False Float=1
         )
        )
        1=&MakeCustomFloatFromBool(
         Bool=&SetBoolByFloatComparison(A=$User.Direction_01, B=0, Comparison Type=::NewEnumerator6)
         True Float=&Add_Float(A=1, B=&Multiply_Float(A=$User.Direction_01, B=-1))
         False Float=1
        )
        2=?
        3=?
        Index=&SpawnGroup()
        Num=::NewEnumerator0
       ))
     )
    )) {isSetParameters=true}
   &SolveForcesAndVelocity(Clamp Velocity=false, Limit Acceleration=false)
   &DynamicMaterialParameters(
    Index 0 Param 1=&FloatFromCurve(FloatCurve=DataInterface() {propertyValues {Curve {keys [
         {arriveTangent=1.2629, interpMode=RCIM_Cubic, leaveTangent=1.2629, tangentMode=RCTM_User, time=-0.0045, value=-1.9799}
         {arriveTangent=4.4591, interpMode=RCIM_Cubic, leaveTangent=4.4591, tangentMode=RCTM_User, time=1, value=0}
        ]}}})
    Index 0 Param 2=&FloatFromCurve(FloatCurve=DataInterface() {propertyValues {Curve {keys [<at 0.0102 4>,<at 0.9749 0>]}}})
    Index 0 Param 3=&MakeFloatFromVector(Vector=$Particles.Scale, Channel=::NewEnumerator0)
    Index 0 Param 4=1
    Index 1 Param 1=&FloatFromCurve(FloatCurve=DataInterface() {propertyValues {Curve {keys [
         {arriveTangent=8.1609, interpMode=RCIM_Cubic, leaveTangent=8.1609, tangentMode=RCTM_User, time=0.7, value=0}
         {arriveTangent=0.3683, interpMode=RCIM_Cubic, leaveTangent=0.3683, tangentMode=RCTM_User, time=1, value=1}
        ]}}})
    Module.Index0Float0Write=true
    Module.Index0Float1Write=true
    Module.Index0Float2Write=true
    Module.Index0Float3Write=true
    Module.Index1Float0Write=true
    Module.Index1Float1Write=true
    Module.Index1Float2Write=true
    Module.Index1Float3Write=true
    Write Parameter Index 1=::NewEnumerator1
   )
   &ScaleColor(
    Scale Alpha=&FloatFromCurve(
     FloatCurve=DataInterface() {propertyValues {Curve {keys [{interpMode=RCIM_Cubic, tangentMode=RCTM_User, time=0, value=0}, {interpMode=RCIM_Cubic, tangentMode=RCTM_User, time=0.16, value=1}]}}}
     Scale Curve=&MergeHelper_FloatSwitch(
      0=&MakeCustomFloatFromBool(
       Bool=&SetBoolByFloatComparison(A=$User.Direction, B=0, Comparison Type=::NewEnumerator6)
       True Float=0
       False Float=1
      )
      1=1
      2=?
      3=?
      Index=&SpawnGroup()
      Num=::NewEnumerator0
     )
    )
    ScaleA=true
    ScaleRGB=true
   )
   &BlendSpace(Rotation Contribution=0.9)
   &ScaleMeshSize(Scale Factor=&Multiply_VectorByFloat(
     Vector=<1 1 1>
     Float=&Lerp_Float(
      A=1
      B=0
      Alpha=&FloatFromCurve(
       FloatCurve=DataInterface() {propertyValues {Curve {defaultValue=1, keys [{tangentMode=RCTM_User, time=0, value=0}, <at 0.4533 0>, {tangentMode=RCTM_User, time=1, value=1}]}}}
       CurveIndex=&RemapRange(
        Input Value To Remap=&DistanceBetweenPositions(A=$Particles.Position, B=$Output.SocketLocation003.SampledPosition)
        Input Min=0
        Input Max=100
        Output Min=0
        Output Max=1
       )
      )
     )
    )) {enabled=false, module=ScaleMeshSize001}
   &ScaleMeshSize(Scale Factor=&Multiply_VectorByFloat(
     Vector=&VectorFromFloat(Value=&MergeHelper_FloatSwitch(
       0=&MakeCustomFloatFromBool(
        Bool=&SetBoolByFloatComparison(A=$User.Direction_01, B=0)
        True Float=&Add_Float(A=1, B=$User.Direction_01)
        False Float=1
       )
       1=&MakeCustomFloatFromBool(
        Bool=&SetBoolByFloatComparison(A=$User.Direction_01, B=0, Comparison Type=::NewEnumerator6)
        True Float=&Add_Float(A=1, B=&Multiply_Float(A=$User.Direction_01, B=-1))
        False Float=1
       )
       2=?
       3=?
       Index=&SpawnGroup()
       Num=::NewEnumerator0
      ))
     Float=&Lerp_Float(
      A=1
      B=0
      Alpha=&FloatFromCurve(
       FloatCurve=DataInterface() {propertyValues {Curve {defaultValue=1, keys [{tangentMode=RCTM_User, time=0, value=0}, <at 0.6 0>, {tangentMode=RCTM_User, time=1, value=1}]}}}
       CurveIndex=&RemapRange(
        Input Value To Remap=&DistanceBetweenPositions(A=$Particles.Position, B=$Output.SocketLocation003.SampledPosition)
        Input Min=0
        Input Max=100
        Output Min=0
        Output Max=1
       )
      )
     )
    ))
  ]
  rendererClasses [&NiagaraMeshRendererProperties]
  renderers [{rendererClass=&NiagaraMeshRendererProperties, rendererIndex=0}]
 }
 Splashes {
  emitterUpdateScript [
   &EmitterState(Life Cycle Mode=::NewEnumerator1, Loop Behavior=::NewEnumerator1, Loop Duration=11.1, UseLoopDelay=false)
   &SpawnBurst_Instantaneous(Spawn Count=60, Spawn Time=0.8, Use Spawn Probability=false, UseLoopCountLimit=false)
   &SpawnBurst_Instantaneous(Spawn Count=60, Spawn Group=1, Spawn Time=6, Use Spawn Probability=false, UseLoopCountLimit=false) {module=SpawnBurst_Instantaneous001}
  ]
  particleSpawnScript [
   &InitializeParticle(
    Color=&FNTimeOfDayColor(
     Night Color=&RandomRangeLinearColor(Minimum=<rgba 0.4344 0.5302 0.8177 0.7043>, Maximum=<rgba 0.27 0.5956 0.9427 0.887>)
     Morning Color=&RandomRangeLinearColor(Minimum=<rgba 0.6442 0.6876 0.8177 1>, Maximum=<rgba 0.27 0.5956 0.9427 1>)
     Day Color=&RandomRangeLinearColor(Minimum=<rgba 0.6981 0.8059 0.9271 1>, Maximum=<rgba 0.4948 0.7636 0.9896 1>)
     Evening Color=&RandomRangeLinearColor(Minimum=<rgba 0.6442 0.6876 0.8177 1>, Maximum=<rgba 0.27 0.5956 0.9427 1>)
    )
    Lifetime Max=0.4
    Lifetime Min=0.2
    Lifetime Mode=::NewEnumerator1
    Mass Mode=::NewEnumerator3
    Sprite Rotation Mode=::NewEnumerator2
    Sprite Size Mode=::NewEnumerator1
    Uniform Sprite Size Max=50
    Uniform Sprite Size Min=15
    UsePositionOffset=false
    Write Material Random=true
   )
   &SkeletalMeshLocation(
    MaskSpawnGroup=false
    Mesh Sampling Type=::NewEnumerator5
    Orientation Sampling=::NewEnumerator1
    Position Sampling=::NewEnumerator1
    Sampled Orientation Coordinate Space=::Local
    Sampled Position Coordinate Space=::Local
    Sampled Velocity Coordinate Space:ENiagaraCoordinateSpace=::Local
    Skeletal Mesh=DataInterface() {propertyValues {
      DefaultMesh=null
      ExcludeBoneName=root
      OptionalComponentClass=null
      PreviewMesh=&FloralMane_Gadget
      SkinningMode=SkinOnTheFly
      SoftSourceActor=null
      SourceMode=Default
      UvSetIndex=0
      WholeMeshLOD=-1
      bExcludeBone=false
      bReadDeformedGeometry=true
      bRequireCurrentFrameData=true
      ComponentTags []
      FilteredBones []
      FilteredSockets [JNT_VFXSocket]
      MeshUserParameter {parameter {name=None, typeDefHandle {registeredTypeIndex=1}, varData []}}
      SamplingRegions []
     }}
    Socket Sampling Mode=::NewEnumerator5
   ) {module=SkeletalMeshLocation001}
   &ShapeLocation(
    Additional Quaternion Rotation=false
    MaskSpawnGroup=false
    Non Uniform Scale=<0.9 1.66 1>
    Override Seed=false
    Ring Radius=28
    Rotation Mode=::NewEnumerator3
    Rotation Quaternion=$Particles.SkeletalMeshLocation001.SampledRotation
    Shape Origin=$Particles.SkeletalMeshLocation001.SampledPosition
    Shape Primitive=::NewEnumerator4
    UseOwnerScale=true
   ) {module=ShapeLocation001}
   &SubUVAnimation(
    End Frame Range Override=3
    Evaluation Type=::NewEnumerator0
    "Sprite Renderer "=DataInterface() {propertyValues {SpriteRenderer=&self:Splashes_0.NiagaraSpriteRendererProperties_1}}
    SubUV Animation Mode:ENiagara_SubUVLookupModeV2=::NewEnumerator2
    UseEndFrame=true
    UseStartFrame=true
   )
   &AddVelocity(Module.Use Falloff Distance=false, Origin Offset=<0 0 -2>, Override Seed=false, Velocity Mode=::NewEnumerator1)
   &AddVelocity(Velocity=&RandomRangeVector(Minimum=<0 0 200>, Maximum=<0 0 400>)) {module=AddVelocity001}
  ]
  particleUpdateScript [
   &ParticleState()
   &ScaleSpriteSize(
    Non-Uniform Curve Sprite Scale=DataInterface() {propertyValues {
      XCurve {keys [
        {arriveTangent=3.4688, interpMode=RCIM_Cubic, leaveTangent=3.4688, tangentMode=RCTM_User, time=-0.0082, value=0.3989}
        {interpMode=RCIM_Cubic, tangentMode=RCTM_User, time=1, value=1}
       ]}
      YCurve {keys [
        {arriveTangent=5.7272, interpMode=RCIM_Cubic, leaveTangent=5.7272, tangentMode=RCTM_User, time=-0.0082, value=0.3989}
        {interpMode=RCIM_Cubic, tangentMode=RCTM_User, time=1, value=1}
       ]}
     }}
    Scale Sprite Size Mode=::NewEnumerator3
   )
   &Drag(Drag=&RandomRangeFloat(Minimum=0.3, Maximum=1.3), Use Linear Drag=true, Use Rotational Drag=true)
   &GravityForce(Gravity=&MakeVector(
     X=0
     Y=0
     Z=&FloatFromCurve(
      FloatCurve=DataInterface() {propertyValues {Curve {keys [<at 0.0754 -821.1881>,<at 0.8013 -1535.7174>]}}}
      Scale Curve=&RandomRangeFloat(Minimum=2.5, Maximum=3)
     )
    ))
   &SolveForcesAndVelocity(Clamp Velocity=false, Limit Acceleration=false)
   &ScaleColor(
    Scale Alpha=&FloatFromCurve(
     FloatCurve=DataInterface() {propertyValues {Curve {
        defaultValue=1
        keys [
         {arriveTangent=-0.4013, interpMode=RCIM_Cubic, leaveTangent=-0.4013, tangentMode=RCTM_User, time=0.0783, value=1.012}
         {arriveTangent=-2.0996, interpMode=RCIM_Cubic, leaveTangent=-2.0996, tangentMode=RCTM_User, time=0.3372, value=0.4125}
         {arriveTangent=-0.0835, tangentMode=RCTM_User, time=0.9653, value=-0.0194}
        ]
       }}}
     Scale Curve=&RandomRangeFloat(Minimum=0.6, Maximum=0.85)
    )
    ScaleA=true
    ScaleRGB=true
   ) {module=ScaleColor001}
   &DynamicMaterialParameters(
    Index 0 Param 1=&FloatFromCurve(FloatCurve=DataInterface() {propertyValues {Curve {keys [
         <at 0.1543 1 RCIM_Cubic>
         {arriveTangent=-2.4268, interpMode=RCIM_Cubic, leaveTangent=-2.4268, tangentMode=RCTM_User, time=0.614, value=0.5527}
         {arriveTangent=0.3036, interpMode=RCIM_Cubic, leaveTangent=0.3036, tangentMode=RCTM_User, time=1, value=0}
        ]}}})
    Module.Index0Float0Write=true
    Module.Index0Float1Write=true
    Module.Index0Float2Write=true
    Module.Index0Float3Write=true
   )
  ]
  rendererClasses [&NiagaraSpriteRendererProperties]
  renderers [{rendererClass=&NiagaraSpriteRendererProperties}]
 }
 WakeLines {
  emitterSpawnScript [
   /SetVariables(Emitter.ParticlesPerTrail=30) {isSetParameters=true}
   /SetVariables(Emitter.Noise Panner Speed:Vector4f=<-0.4 0 0 0>) {isSetParameters=true}
  ]
  emitterUpdateScript [
   &EmitterState()
   &SpawnBurst_Instantaneous(
    Spawn Count=&Multiply_Int(A=$Emitter.ParticlesPerTrail, B=2)
    Use Spawn Probability=false
    UseLoopCountLimit=true
   )
  ]
  particleSpawnScript [
   &PartitionParticles(
    Normalized Index In Partition Includes One=true
    Particles Per Partition:NiagaraInt32=$Emitter.ParticlesPerTrail
    Partition Mode:ENiagara_PartitionMode_v2=::NewEnumerator1
    Write Partition Index to Ribbon Link Order:ENiagara_AttributeSamplingApplyOutput=::NewEnumerator0
    Write Partition to Ribbon ID=::NewEnumerator0
   )
   &InitializeParticle(
    Color=&FNTimeOfDayColor(
     Night Color=<rgba 0.542 0.7243 1 0.3565>
     Morning Color=<rgba 0.8796 0.9164 1 0.7478>
     Day Color=<rgba 1.0555 1.0997 1.2 0.7826>
     Evening Color=<rgba 0.8796 0.9164 1 0.7478>
    )
    Lifetime=11.1
    UsePositionOffset=false
    Write Material Random=true
   )
   /SetVariables() {isSetParameters=true}
   /SetVariables(Particles.GizmoScale:Vector3f=<5 5 5>) {isSetParameters=true}
   /SetVariables(Particles.NormalizedParticleIndex:NiagaraFloat=&Power(a=&OneMinusFloat(Float=$Particles.PartitionParticles.Normalized Index In Partition), b=13)) {isSetParameters=true}
  ]
  particleUpdateScript [
   &ParticleState(Kill Particles When Lifetime Has Elapsed=false, Let Infinitely Lived Particles Die When Emitter Deactivates=true)
   /SetVariables(
    Particles.RibbonFacing=&TransformVector(Vector=<0 0 1>, Source Space=::World, Destination Space=::Local)
    Particles.RibbonWidth=&Lerp_Float(
     A=35
     B=9
     Alpha=&FloatFromCurve(
      FloatCurve=DataInterface() {propertyValues {Curve {keys [{arriveTangent=-2.079, interpMode=RCIM_Cubic, leaveTangent=-2.079, tangentMode=RCTM_User, time=0, value=1}, <at 0.7 0 RCIM_Cubic>]}}}
      CurveIndex=$Particles.PartitionParticles.Normalized Index In Partition
     )
    )
   ) {isSetParameters=true}
   /SetVariables(Particles.PreviousPos:NiagaraPosition=$Particles.Position) {isSetParameters=true}
   &SkeletalMeshLocation(
    MaskSpawnGroup=false
    Mesh Sampling Type=::NewEnumerator5
    Sampled Position Offset=&Add_Vector(
     A=&MakeVector(
      X=&Add_Float(
       A=&ClampFloat(
        Float=&Multiply_Float(
         A=$Particles.NormalizedParticleIndex
         B=&RemapRange(Input Value To Remap=$Particles.Roadbending, Input Min=-250, Input Max=250, Output Min=-100, Output Max=100)
        )
        Min=-100
        Max=100
       )
       B=&Multiply_Float(
        A=&Sine(
         Normalized Angle=&Multiply_Float(
          A=&Add_Float(A=$System.Age, B=&Multiply_Float(A=$Particles.PartitionParticles.Normalized Index In Partition, B=-1))
          B=&MergeHelper_FloatSwitch(0=1, 1=-1, 2=?, 3=?, Index=$Particles.PartitionParticles.Partition, Num=::NewEnumerator0)
         )
         Period=1
         Scale=1
         Bias=0
        )
        B=&FloatFromCurve(
         FloatCurve=DataInterface() {propertyValues {Curve {
            defaultValue=1
            keys [
             {arriveTangent=3.7882, interpMode=RCIM_Cubic, leaveTangent=3.7882, tangentMode=RCTM_User, time=0, value=0}
             {arriveTangent=-0.9674, interpMode=RCIM_Cubic, leaveTangent=-0.9674, tangentMode=RCTM_User, time=0.2302, value=0.3882}
             {arriveTangent=2.3649, interpMode=RCIM_Cubic, leaveTangent=2.3649, tangentMode=RCTM_User, time=0.5469, value=0.5793}
             {interpMode=RCIM_Cubic, tangentMode=RCTM_User, time=1, value=1}
            ]
           }}}
         CurveIndex=&ReturnNormalizedExecIndex(Normalized Index Scale=0.8)
         Scale Curve=15
        )
       )
      )
      Y=&Add_Float(
       A=-60
       B=&Add_Float(A=60, B=&Multiply_Float(A=$Particles.PartitionParticles.Normalized Index In Partition, B=-150))
      )
      Z=0
     )
     B=&Add_Vector(
      A=&AddVectorToPosition(Position=Linked() {linkedVariable {name=System.MasterRoadOffset, type=Vector3f}}, Vector=<0 0 0>)
      B=&MakeVector(
       X=&Lerp_Float(
        A=&MergeHelper_FloatSwitch(0=-65, 1=65, 2=?, 3=?, Index=$Particles.PartitionParticles.Partition, Num=::NewEnumerator0)
        B=&Lerp_Float(
         A=&MergeHelper_FloatSwitch(0=-19, 1=19, 2=?, 3=?, Index=$Particles.PartitionParticles.Partition, Num=::NewEnumerator0)
         B=&MergeHelper_FloatSwitch(
          0=&Lerp_Float(
           A=-40
           B=-23
           Alpha=&FloatFromCurve(FloatCurve=DataInterface() {propertyValues {Curve {keys [<at 0 0.5279>,<at 0.0845 0.5279>,<at 0.1521 0.3643>,<at 0.1906 0.5176>,<at 0.2315 1>,<at 0.2975 1>,<at 0.3321 0.4549>,<at 0.37 1>,<at 1 1>]}}})
          )
          1=&Lerp_Float(
           A=40
           B=23
           Alpha=&FloatFromCurve(FloatCurve=DataInterface() {propertyValues {Curve {keys [<at 0 0.5279>,<at 0.0845 0.5279>,<at 0.1521 0.3643>,<at 0.1906 0.5176>,<at 0.2315 1>,<at 0.2975 1>,<at 0.3321 0.4549>,<at 0.37 1>,<at 1 1>]}}})
          )
          2=?
          3=?
          Index=$Particles.PartitionParticles.Partition
          Num=::NewEnumerator0
         )
         Alpha=&RemapRange(
          Input Value To Remap=$Particles.Roadbending
          Input Min=0
          Input Max=&MergeHelper_FloatSwitch(0=-100, 1=100, 2=?, 3=?, Index=$Particles.PartitionParticles.Partition, Num=::NewEnumerator0)
          Output Min=1
          Output Max=0
         )
        )
        Alpha=&FloatFromCurve(
         FloatCurve=DataInterface() {propertyValues {Curve {defaultValue=1, keys [{tangentMode=RCTM_User, time=0, value=1}, {tangentMode=RCTM_User, time=1, value=0}]}}}
         CurveIndex=$Particles.PartitionParticles.Normalized Index In Partition
        )
       )
       Y=0
       Z=3
      )
     )
    )
    Sampled Position Offset Coordinate Space:ENiagaraSkelMeshTransforms=::NewEnumerator2
    Skeletal Mesh=$System.Skeletal Mesh Reader
    Socket Sampling Mode=::NewEnumerator5
    Use Sampled Mesh Attribute as Ribbon ID=false
   )
   &self:SetRoadBending_02(CurrentPos:NiagaraPosition=$Particles.Position, MaxBending=5, RestSpeed=1, VelDivider=5) {enabled=false, module=SetRoadBending_02}
   &ScaleColor(
    Scale RGB=&VectorFromFloat(Value=&RemapRange(
      Input Value To Remap=&Sine(
       Normalized Angle=&Add_Float(
        A=&Multiply_Float(A=$Particles.PartitionParticles.Normalized Index In Partition, B=3)
        B=&Multiply_Float(
         A=&Add_Float(A=$Engine.System.Age, B=&Multiply_Float(A=$System.FN_Integrator002.Value, B=10))
         B=-1
        )
       )
       Period=1
       Scale=1
       Bias=0
      )
      Input Min=-1
      Input Max=1
      Output Min=0.1
      Output Max=1
     ))
    ScaleA=true
    ScaleRGB=true
   )
   &FN_Sprite_Attributes(
    Color=&MultiplyLinearColor(
     Linear Color=<rgba 0.8123 1.0591 1.2 1>
     Scale Factor=&MakeVector4FromVectorAndFloat(
      "Vector (RGB)"=<1 1 1>
      "Float (Alpha)"=&Multiply_Float(
       A=1
       B=&FloatFromCurve(
        FloatCurve=DataInterface() {propertyValues {Curve {keys [{arriveTangent=-4.0505, interpMode=RCIM_Cubic, leaveTangent=-4.0505, tangentMode=RCTM_User, time=0.3, value=1}, <at 1 0>]}}}
        CurveIndex=$Particles.PartitionParticles.Normalized Index In Partition
        Scale Curve=0.6
       )
      )
     )
    )
    Create Color Attribute=true
    Write to Module Namespace=true
   )
   &DynamicMaterialParameters(Module.Index0Float0Write=true, Module.Index0Float1Write=false, Module.Index0Float2Write=false, Module.Index0Float3Write=false)
  ]
  rendererClasses [&NiagaraRibbonRendererProperties]
  renderers [
   {rendererClass=&NiagaraRibbonRendererProperties}
   {rendererClass=&NiagaraRibbonRendererProperties, rendererIndex=1}
  ]
 }
 WaterTrail {
  emitterUpdateScript [
   &EmitterState()
   &SpawnRate(SpawnRate=50, Use Spawn Probability=false)
  ]
  particleSpawnScript [
   &InitializeParticle(
    Color=&FNTimeOfDayColor(
     Night Color=<rgba 0.1139 0.9251 2 1>
     Morning Color=<rgba 0.1139 0.9251 2 1>
     Day Color=<rgba 0.1139 0.9251 2 1>
     Evening Color=<rgba 0.1139 0.9251 2 1>
    )
    Lifetime=0.5
    Ribbon Facing Mode=::NewEnumerator1
    Ribbon Facing Vector:Vector3f=$Engine.Owner.SystemZAxis
    Ribbon Width=185
    Ribbon Width Mode=::NewEnumerator1
    UsePositionOffset=false
    Write Material Random=true
   )
   &SocketLocation(
    SamplePosOffset=true
    Sampled Position Offset=<0 50 -2>
    Sampled Position Offset Coordinate Space:ENiagaraSkelMeshTransforms=::NewEnumerator2
    Socket Reader=$System.FloralMane Sockets
    Socket Sampling Mode=::NewEnumerator3
    VelocitySpeedLimit=false
   )
   &AddVelocity(Velocity=&MakeVector(X=0, Y=-1000, Z=0))
   /SetVariables(Particles.Position=&Add_Vector(
     A=<0 0 3>
     B=&Multiply_Vector(A=Linked() {linkedVariable {name=StackContext.Position, type=NiagaraPosition}}, B=<1 1 0>)
    )) {isSetParameters=true}
  ]
  particleUpdateScript [
   &ParticleState()
   &ScaleColor(
    Scale Alpha=&FloatFromCurve(
     FloatCurve=DataInterface() {propertyValues {Curve {keys [
         {interpMode=RCIM_Cubic, tangentMode=RCTM_Break, time=0, value=0}
         {arriveTangent=1.6363, interpMode=RCIM_Cubic, leaveTangent=1.6363, time=0.15, value=1}
         {arriveTangent=-1.5385, interpMode=RCIM_Cubic, leaveTangent=-1.5385, time=0.4871, value=0.797}
         {interpMode=RCIM_Cubic, tangentMode=RCTM_Break, time=0.8, value=0}
        ]}}}
     Scale Curve=0.7
    )
    Scale RGB=&MakeVectorFromLinearColorRGB(Color=&ColorFromCurve(ColorCurve=DataInterface() {propertyValues {
        AlphaCurve {keys [<at 0 1 RCIM_Cubic>]}
        BlueCurve {keys [<at 0.0417 5>,<at 0.0964 1.25 RCIM_Cubic>]}
        GreenCurve {keys [<at 0.0417 5>,<at 0.0964 1.25 RCIM_Cubic>]}
        RedCurve {keys [<at 0.0417 5>,<at 0.0964 1.25 RCIM_Cubic>]}
       }}))
    ScaleA=true
    ScaleRGB=true
   )
   &ScaleRibbonWidth(Ribbon Width Scale=&FloatFromCurve(FloatCurve=DataInterface() {propertyValues {Curve {keys [
         <at 0.0009 0.3 RCIM_Cubic>
         {arriveTangent=2.4405, interpMode=RCIM_Cubic, leaveTangent=2.4405, tangentMode=RCTM_User, time=0.2248, value=0.921}
         {arriveTangent=-1.171, interpMode=RCIM_Cubic, leaveTangent=-1.171, time=0.6741, value=1.0578}
         <at 1.0122 -0.0012 RCIM_Cubic>
        ]}}}))
   &Drag(Drag=3, Use Linear Drag=true, Use Rotational Drag=true)
   &GravityForce(Gravity=<-500 0 0>)
   &SolveForcesAndVelocity(Clamp Velocity=false, Limit Acceleration=false)
   &BlendSpace(Rotation Contribution=0)
   /SetVariables(Particles.Wake Outline Width:NiagaraFloat=&Multiply_Float(
     A=&Add_Float(
      A=160
      B=&Lerp_Float(
       A=120
       B=0
       Alpha=&FloatFromCurve(
        FloatCurve=DataInterface() {propertyValues {Curve {keys [
            <at 0.0035 1 RCIM_Cubic>
            {arriveTangentWeight=0.2525, interpMode=RCIM_Cubic, leaveTangentWeight=0.0934, tangentMode=RCTM_User, time=0.1371, value=0.6531}
            {arriveTangent=-0.4139, interpMode=RCIM_Cubic, leaveTangent=-0.4139, time=0.232, value=1.432}
            {arriveTangent=-1.2259, interpMode=RCIM_Cubic, leaveTangent=-1.2259, time=0.3379, value=0.57}
            {arriveTangentWeight=0.0177, interpMode=RCIM_Cubic, leaveTangentWeight=0.3014, tangentMode=RCTM_User, time=0.3749, value=1.2568}
            <at 1 1 RCIM_Cubic>
           ]}}}
        CurveIndex=$Emitter.NormalizedLoopAge
       )
      )
     )
     B=&FloatFromCurve(FloatCurve=DataInterface() {propertyValues {Curve {
         defaultValue=1
         keys [
          <at -0.0018 0.2235 RCIM_Cubic>
          {arriveTangent=1.9234, interpMode=RCIM_Cubic, leaveTangent=1.9234, time=0.1213, value=0.5835}
          {arriveTangent=0.8499, interpMode=RCIM_Cubic, leaveTangent=0.8499, time=0.2083, value=0.6275}
          {arriveTangent=0.1196, interpMode=RCIM_Cubic, leaveTangent=0.1196, time=0.2775, value=0.7163}
          {arriveTangent=0.4531, interpMode=RCIM_Cubic, leaveTangent=0.4531, time=0.4624, value=0.6579}
          {arriveTangent=0.5513, interpMode=RCIM_Cubic, leaveTangent=0.5513, time=0.6262, value=0.8743}
          {arriveTangent=0.3363, interpMode=RCIM_Cubic, leaveTangent=0.3363, time=0.8088, value=0.8489}
          {arriveTangent=1.3622, interpMode=RCIM_Cubic, leaveTangent=1.3622, tangentMode=RCTM_User, time=1, value=1}
         ]
        }}})
    )) {isSetParameters=true}
   /SetVariables(Particles.Wake Outline Color:LinearColor=&MakeLinearColorFromVectorAndFloat(
     "Vector (RGB)"=&MakeVectorFromLinearColorRGB(Color=&FNTimeOfDayColor(
       Night Color=&ColorFromCurve(ColorCurve=DataInterface() {propertyValues {
          AlphaCurve {keys [<at 0.7797 1>,<at 1 0>]}
          BlueCurve {keys [
            <at 0.1179 1>
            <at 0.1781 1.35>
            <at 0.2855 1.15>
            {arriveTangent=-0.2113, interpMode=RCIM_Cubic, leaveTangent=-0.2113, time=0.3872, value=1}
            <at 0.9955 1>
           ]}
          GreenCurve {keys [
            <at 0.1179 0.5486>
            <at 0.1781 1.1736>
            <at 0.2855 0.9997>
            {arriveTangent=-0.8299, interpMode=RCIM_Cubic, leaveTangent=-0.8299, time=0.3872, value=0.8258}
            <at 0.9955 0.4105>
           ]}
          RedCurve {keys [
            <at 0.1179 0>
            <at 0.1781 1.1391>
            <at 0.2855 0.9704>
            {arriveTangent=-1.3667, interpMode=RCIM_Cubic, leaveTangent=-1.3667, time=0.3872, value=0.3544}
            <at 0.9955 0>
           ]}
         }})
       Morning Color=&ColorFromCurve(ColorCurve=DataInterface() {propertyValues {
          AlphaCurve {keys [<at 0.7797 1>,<at 1 0>]}
          BlueCurve {keys [
            <at 0.1179 1>
            <at 0.1781 1.5>
            <at 0.2855 1.3>
            {arriveTangent=-0.4225, interpMode=RCIM_Cubic, leaveTangent=-0.4225, time=0.3872, value=1}
            <at 0.9955 1>
           ]}
          GreenCurve {keys [
            <at 0.1179 0.5486>
            <at 0.1781 1.304>
            <at 0.2855 1.1301>
            {arriveTangent=-1.0136, interpMode=RCIM_Cubic, leaveTangent=-1.0136, time=0.3872, value=0.8258}
            <at 0.9955 0.4105>
           ]}
          RedCurve {keys [
            <at 0.1179 0>
            <at 0.1781 1.2657>
            <at 0.2855 1.0969>
            {arriveTangent=-1.545, interpMode=RCIM_Cubic, leaveTangent=-1.545, time=0.3872, value=0.3544}
            <at 0.9955 0>
           ]}
         }})
       Day Color=&ColorFromCurve(ColorCurve=DataInterface() {propertyValues {
          AlphaCurve {keys [<at 0.7797 1>,<at 1 0>]}
          BlueCurve {keys [
            <at 0.1179 1>
            <at 0.1781 1.5>
            <at 0.2855 1.3>
            {arriveTangent=-0.4225, interpMode=RCIM_Cubic, leaveTangent=-0.4225, time=0.3872, value=1}
            <at 0.9955 1>
           ]}
          GreenCurve {keys [
            <at 0.1179 0.5486>
            <at 0.1781 1.304>
            <at 0.2855 1.1301>
            {arriveTangent=-1.0136, interpMode=RCIM_Cubic, leaveTangent=-1.0136, time=0.3872, value=0.8258}
            <at 0.9955 0.4105>
           ]}
          RedCurve {keys [
            <at 0.1179 0>
            <at 0.1781 1.2657>
            <at 0.2855 1.0969>
            {arriveTangent=-1.545, interpMode=RCIM_Cubic, leaveTangent=-1.545, time=0.3872, value=0.3544}
            <at 0.9955 0>
           ]}
         }})
       Evening Color=&ColorFromCurve(ColorCurve=DataInterface() {propertyValues {
          AlphaCurve {keys [<at 0.7797 1>,<at 1 0>]}
          BlueCurve {keys [
            <at 0.1179 1>
            <at 0.1781 1.5>
            <at 0.2855 1.3>
            {arriveTangent=-0.4225, interpMode=RCIM_Cubic, leaveTangent=-0.4225, time=0.3872, value=1}
            <at 0.9955 1>
           ]}
          GreenCurve {keys [
            <at 0.1179 0.5486>
            <at 0.1781 1.304>
            <at 0.2855 1.1301>
            {arriveTangent=-1.0136, interpMode=RCIM_Cubic, leaveTangent=-1.0136, time=0.3872, value=0.8258}
            <at 0.9955 0.4105>
           ]}
          RedCurve {keys [
            <at 0.1179 0>
            <at 0.1781 1.2657>
            <at 0.2855 1.0969>
            {arriveTangent=-1.545, interpMode=RCIM_Cubic, leaveTangent=-1.545, time=0.3872, value=0.3544}
            <at 0.9955 0>
           ]}
         }})
      ))
     "Float (Alpha)"=&FloatFromCurve(
      FloatCurve=DataInterface() {propertyValues {Curve {keys [
          {interpMode=RCIM_Cubic, tangentMode=RCTM_Break, time=0, value=0}
          {arriveTangent=1.6363, interpMode=RCIM_Cubic, leaveTangent=1.6363, time=0.15, value=1}
          {arriveTangent=-1.5385, interpMode=RCIM_Cubic, leaveTangent=-1.5385, time=0.4871, value=0.797}
          {interpMode=RCIM_Cubic, tangentMode=RCTM_Break, time=0.8, value=0}
         ]}}}
      Scale Curve=0.5
     )
    )) {isSetParameters=true}
  ]
  rendererClasses [&NiagaraRibbonRendererProperties]
  renderers [
   {rendererClass=&NiagaraRibbonRendererProperties}
   {rendererClass=&NiagaraRibbonRendererProperties, rendererIndex=1}
  ]
 }
 boat_ripples {
  emitterUpdateScript [
   &EmitterState()
   &SpawnBurst_Instantaneous(Use Spawn Probability=false, UseLoopCountLimit=false)
   &SpawnRate(SpawnRate=2, Use Spawn Probability=false)
  ]
  particleSpawnScript [
   &InitializeParticle(
    Color=&FNTimeOfDayColor(
     Night Color=<rgba 0.2409 0.6933 1.2 0.7739>
     Morning Color=<rgba 0.271 0.7799 1.35 0.8957>
     Day Color=<rgba 0.5664 1.1315 1.5 0.9043>
     Evening Color=<rgba 0.271 0.7799 1.35 0.8957>
    )
    Lifetime=&RandomRangeFloat(Minimum=1.5, Maximum=2)
    Mesh Scale=<0.78 0.59 0.05>
    Mesh Scale Mode=::NewEnumerator4
    Position Offset=<0 0 -50>
    UsePositionOffset=false
    Write Material Random=true
   )
   &FN_SetSpawnGroup(Number of Spawn Groups=1)
   &DynamicMaterialParameters(
    Index 0 Param 1=&RandomRangeFloat(Minimum=0, Maximum=1)
    Index 0 Param 3=1
    Module.Index0Param0Write=true
    Module.Index0Param1Write=true
    Module.Index0Param2Write=true
    Module.Index0Param3Write=false
   )
   &AddVelocity(Velocity=<0 -50 0>)
   &CameraOffset(Camera Offset Amount=-5)
  ]
  particleUpdateScript [
   &ParticleState()
   &SocketLocation(
    SamplePosOffset=true
    Sampled Position Offset=<0 15 3>
    Socket Reader=$System.FloralMane Sockets
    Socket Sampling Mode=::NewEnumerator3
    VelocitySpeedLimit=false
   )
   &ScaleColor(
    Scale Alpha=&FloatFromCurve(FloatCurve=DataInterface() {propertyValues {Curve {keys [
         {interpMode=RCIM_Cubic, tangentMode=RCTM_User, time=0, value=0}
         {interpMode=RCIM_Cubic, tangentMode=RCTM_User, time=0.5, value=1}
         {tangentMode=RCTM_User, time=1, value=0}
        ]}}})
    Scale RGB=&VectorFromFloat(Value=&FNTimeOfDayScalar(Night Scale=1.5, Morning Scale=1.3, Day Scale=1.5, Evening Scale=1.2))
    ScaleA=true
    ScaleRGB=true
   )
   &InterpolateOverTime(
    Rate Of Change=8
    Target Vector Value:Vector3f=&AddVectorToPosition(
     Position=$Particles.Position
     Vector=&RotateVectorByQuaternion(Vector To Rotate=<0 -100 0>, Quaternion=$Particles.SkeletalMeshLocation.SampledRotation)
    )
    Variable Type=::NewEnumerator2
   )
   &ScaleMeshSize(Scale Factor=&VectorFromCurve(
     VectorCurve=DataInterface() {propertyValues {
       XCurve {keys [{arriveTangent=1.898, interpMode=RCIM_Cubic, leaveTangent=1.898, tangentMode=RCTM_User, time=0, value=1}]}
       YCurve {keys [{arriveTangent=0.5312, interpMode=RCIM_Cubic, leaveTangent=0.5312, tangentMode=RCTM_User, time=0, value=1}, <at 1 1.2>]}
       ZCurve {keys [{arriveTangent=9.0725, interpMode=RCIM_Cubic, leaveTangent=9.0725, tangentMode=RCTM_User, time=0, value=1}]}
      }}
     Scale Curve=<1 1 1.5>
    )) {module=ScaleMeshSize001}
   &SolveForcesAndVelocity(Clamp Velocity=false, Limit Acceleration=false)
  ]
  rendererClasses [&NiagaraMeshRendererProperties]
  renderers [{rendererClass=&NiagaraMeshRendererProperties, rendererIndex=0}]
 }
}
summary {
 systemName=NS_Emote_FloralMane_PropLooping
 emitters [
  {emitterName=WaterTrail, rendererClasses [&NiagaraRibbonRendererProperties]}
  {emitterName=Splashes, rendererClasses [&NiagaraSpriteRendererProperties]}
  {emitterName=boat_ripples, rendererClasses [&NiagaraMeshRendererProperties]}
  {emitterName=WakeLines, rendererClasses [&NiagaraRibbonRendererProperties]}
  {emitterName=Edge_Spray_Mesh005, simTarget=GPUComputeSim, rendererClasses [&NiagaraMeshRendererProperties]}
  {emitterName=Edge_Splashes, simTarget=GPUComputeSim, rendererClasses [&NiagaraSpriteRendererProperties]}
 ]
 userVariables [
  User.bIsFrontEnd {type=NiagaraBool}
  User.bIsFrontEndPreview {type=NiagaraBool}
  User.Direction {type=NiagaraFloat, defaultValue=0, description=""}
  User.Direction_01 {type=NiagaraFloat, defaultValue=0, description=""}
 ]
}
