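# Niagara system for the spinning part of the FloralMane emote, producing water-like splashes around the gadget prop.
# A single Splashes emitter samples the gadget's VFX socket, and spawn groups set the visibility tag between sprite and mesh renderers.
# Masked inputs give one group time-of-day tinted, stretched-ring spray with velocity, drag and gravity, while the other stays in place and fades.
# Curves drive scale, alpha and radial material parameters.
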NiagaraSystem /BRCosmetics/Animation/Game/MainPlayer/Emotes/FloralMane/VFX/NS_Emote_FloralMane_Spinning
dependencies {
 usedDataInterfaces [
  &NiagaraDataInterfaceSkeletalMesh
  &NiagaraDataInterfaceSpriteRendererInfo
  &NiagaraDataInterfaceCurve
  &NiagaraDataInterfaceVector2DCurve
  &NiagaraDataInterfaceSocketReader
 ]
 usedDynamicInputs []
 usedModules [
  &SystemState
  &self:SystemSpawnScript.SystemScriptSource.SystemScriptGraph.NiagaraNodeAssignment_0.SetVariables_#1
 ]
}
emitters {Splashes {
  emitterUpdateScript [
   &EmitterState(Life Cycle Mode=::NewEnumerator1, Loop Behavior=::NewEnumerator1, Loop Duration=0.7, UseLoopDelay=false)
   &SpawnBurst_Instantaneous(Spawn Count=65, Use Spawn Probability=false, UseLoopCountLimit=false)
   &SpawnRate(SpawnRate=150, Use Spawn Probability=false)
   &SpawnBurst_Instantaneous(Spawn Group=1, Use Spawn Probability=false, UseLoopCountLimit=false) {module=SpawnBurst_Instantaneous001}
  ]
  particleSpawnScript [
   /SetVariables(Particles.VisibilityTag=&SpawnGroup()) {isSetParameters=true}
   &InitializeParticle(
    Color=&MaskLinearColor(
     Mask Source=0
     Mask Parameter=?
     Mask=0
     Masked Linear Color=&FNTimeOfDayColor(
      Night Color=&RandomRangeLinearColor(Minimum=<rgba 0.3676 0.6501 0.9167 0.7043>, Maximum=<rgba 0.27 0.5956 0.9427 0.8609>)
      Morning Color=&RandomRangeLinearColor(Minimum=<rgba 0.6442 0.6876 0.8177 1>, Maximum=<rgba 0.27 0.5956 0.9427 1>)
      Day Color=&RandomRangeLinearColor(Minimum=<rgba 0.8776 0.9056 0.9896 1>, Maximum=<rgba 0.383 0.6539 0.9427 1>)
      Evening Color=&RandomRangeLinearColor(Minimum=<rgba 0.6442 0.6876 0.8177 1>, Maximum=<rgba 0.27 0.5956 0.9427 1>)
     )
     Passthrough Linear Color=<rgba 0.4601 0.6396 0.8333 1.3>
    )
    Lifetime=&MaskFloat(
     Mask Source=0
     Mask Parameter=?
     Mask=0
     Masked Float=&RandomRangeFloat(Minimum=0.2, Maximum=0.4)
     Passthrough Float=1.25
    )
    Mass Mode=::NewEnumerator3
    Mesh Scale Mode=::NewEnumerator3
    Mesh Uniform Scale=0.04
    Sprite Rotation Mode=::NewEnumerator2
    Sprite Size Mode=::NewEnumerator1
    Uniform Sprite Size Max=50
    Uniform Sprite Size Min=20
    UsePositionOffset=false
    Write Material Random=true
   )
   &SkeletalMeshLocation(
    MaskSpawnGroup=false
    Mesh Sampling Type=::NewEnumerator5
    Orientation Sampling=::NewEnumerator1
    Position Sampling=::NewEnumerator1
    Sampled Orientation Coordinate Space=::Local
    Sampled Position Coordinate Space=::Local
    Sampled Velocity Coordinate Space:ENiagaraCoordinateSpace=::Local
    Skeletal Mesh=DataInterface() {propertyValues {
      DefaultMesh=null
      ExcludeBoneName=root
      OptionalComponentClass=null
      PreviewMesh=/BRCosmetics/Animation/Props/Gadgets_FloralMane/Meshes/FloralMane_Gadget
      SkinningMode=SkinOnTheFly
      SoftSourceActor=null
      SourceMode=Default
      UvSetIndex=0
      WholeMeshLOD=-1
      bExcludeBone=false
      bReadDeformedGeometry=true
      bRequireCurrentFrameData=true
      ComponentTags []
      FilteredBones []
      FilteredSockets [JNT_VFXSocket]
      MeshUserParameter {parameter {name=None, typeDefHandle {registeredTypeIndex=1}, varData []}}
      SamplingRegions []
     }}
    Socket Sampling Mode=::NewEnumerator5
   ) {module=SkeletalMeshLocation001}
   &InitialMeshOrientation(Mesh Axis To Orient=<0 1 0>, Mesh Orientation Mode=::NewEnumerator2, Orientation Vector=<0 0 1>, Use Rotation Vector=false)
   &ShapeLocation(
    Additional Quaternion Rotation=false
    MaskSpawnGroup=false
    Non Uniform Scale=<1 3.5 1>
    Override Seed=false
    Ring Radius=&MaskFloat(Mask Source=0, Mask Parameter=?, Mask=0, Masked Float=15, Passthrough Float=0)
    Rotation Mode=::NewEnumerator3
    Rotation Quaternion=$Particles.SkeletalMeshLocation001.SampledRotation
    Shape Origin=$Particles.SkeletalMeshLocation001.SampledPosition
    Shape Primitive=::NewEnumerator4
    UseOwnerScale=true
   ) {module=ShapeLocation001}
   &SubUVAnimation(
    End Frame Range Override=3
    Evaluation Type=::NewEnumerator0
    Output Alternate Renderer Binding=true
    "Sprite Renderer "=DataInterface() {propertyValues {SpriteRenderer=&self:Splashes_0.NiagaraSpriteRendererProperties_1}}
    SubUV Animation Mode:ENiagara_SubUVLookupModeV2=::NewEnumerator2
    UseEndFrame=true
    UseStartFrame=true
   )
   &AddVelocity(
    Module.Use Falloff Distance=false
    Origin Offset=<0 0 -2>
    Override Seed=false
    Velocity Mode=::NewEnumerator1
    Velocity Speed=&MaskFloat(Mask Source=0, Mask Parameter=?, Mask=0, Masked Float=100, Passthrough Float=0)
   )
   &AddVelocity(Velocity=&MaskVector(
     Mask Source=0
     Mask Parameter=?
     Mask=0
     Masked Vector=&RandomRangeVector(Minimum=<0 0 100>, Maximum=<0 0 140>)
     Passthrough Vector=<0 0 0>
    )) {module=AddVelocity001}
  ]
  particleUpdateScript [
   &ParticleState()
   &ScaleSpriteSize(
    Non-Uniform Curve Sprite Scale=DataInterface() {propertyValues {
      XCurve {keys [
        {arriveTangent=3.4688, interpMode=RCIM_Cubic, leaveTangent=3.4688, tangentMode=RCTM_User, time=-0.0082, value=0.3989}
        {interpMode=RCIM_Cubic, tangentMode=RCTM_User, time=1, value=1}
       ]}
      YCurve {keys [
        {arriveTangent=5.7272, interpMode=RCIM_Cubic, leaveTangent=5.7272, tangentMode=RCTM_User, time=-0.0082, value=0.3989}
        {interpMode=RCIM_Cubic, tangentMode=RCTM_User, time=1, value=1}
       ]}
     }}
    Scale Sprite Size Mode=::NewEnumerator3
   )
   &ScaleMeshSize(Scale Factor=&VectorFromFloat(Value=&FloatFromCurve(
      FloatCurve=DataInterface() {propertyValues {Curve {
         defaultValue=1
         keys [
          {arriveTangent=2.9253, interpMode=RCIM_Cubic, leaveTangent=2.9253, tangentMode=RCTM_User, time=0.003, value=0.3971}
          {arriveTangent=0.8054, interpMode=RCIM_Cubic, leaveTangent=0.8054, time=0.3598, value=1.0164}
          {arriveTangentWeight=0.7014, interpMode=RCIM_Cubic, tangentMode=RCTM_User, time=1, value=1.2}
         ]
        }}}
      Scale Curve=2.32
     )))
   &Drag(
    Drag=&MaskFloat(
     Mask Source=0
     Mask Parameter=?
     Mask=0
     Masked Float=&RandomRangeFloat(Minimum=0.3, Maximum=1.3)
     Passthrough Float=0
    )
    Use Linear Drag=true
    Use Rotational Drag=true
   )
   &FN_Radial_Material(
    Base Sprite Tile U=3
    Depth Fade=1
    Radial Mask Power=3
    Sprite Extrusion Speed:NiagaraFloat=&FloatFromCurve(FloatCurve=DataInterface() {propertyValues {Curve {keys [
         {interpMode=RCIM_Cubic, tangentMode=RCTM_User, time=0, value=1.5}
         {arriveTangent=-1.1, interpMode=RCIM_Cubic, leaveTangent=-1.1, time=0.1592, value=0.7203}
         {interpMode=RCIM_Cubic, tangentMode=RCTM_User, time=1, value=0.4}
        ]}}})
    Sprite Spin Speed:NiagaraFloat=&FloatFromCurve(FloatCurve=DataInterface() {propertyValues {Curve {keys [<at 0 -1.5>, {arriveTangent=1.5942, leaveTangent=1.5942, time=1, value=-0.2}]}}})
    Write to Module Namespace=true
   )
   &GravityForce(Gravity=&MaskVector(Mask Source=0, Mask Parameter=?, Mask=0, Masked Vector=<0 0 -1000>, Passthrough Vector=<0 0 0>))
   &SolveForcesAndVelocity(Clamp Velocity=false, Limit Acceleration=false)
   &ScaleColor(
    Scale Alpha=&MaskFloat(
     Mask Source=0
     Mask Parameter=?
     Mask=0
     Masked Float=&FloatFromCurve(
      FloatCurve=DataInterface() {propertyValues {Curve {
         defaultValue=1
         keys [
          {interpMode=RCIM_Cubic, tangentMode=RCTM_User, time=0.0155, value=0}
          {interpMode=RCIM_Cubic, tangentMode=RCTM_User, time=0.5743, value=1.0194}
          {arriveTangent=-0.0835, tangentMode=RCTM_User, time=0.9969, value=-0.0194}
         ]
        }}}
      Scale Curve=&RandomRangeFloat(Minimum=0.6, Maximum=0.85)
     )
     Passthrough Float=&FloatFromCurve(FloatCurve=DataInterface() {propertyValues {Curve {keys [
          <at -0.0046 1.0113 RCIM_Cubic>
          {arriveTangent=-1.3263, interpMode=RCIM_Cubic, leaveTangent=-1.3263, time=0.3771, value=0.2858}
          <at 0.7564 0.002 RCIM_Cubic>
         ]}}})
    )
    ScaleA=true
    ScaleRGB=true
   ) {module=ScaleColor001}
   &DynamicMaterialParameters(
    Index 0 Param 1=&FloatFromCurve(FloatCurve=DataInterface() {propertyValues {Curve {keys [
         <at 0.1543 1 RCIM_Cubic>
         {arriveTangent=-2.4268, interpMode=RCIM_Cubic, leaveTangent=-2.4268, tangentMode=RCTM_User, time=0.798, value=0.3996}
         {arriveTangent=0.3036, interpMode=RCIM_Cubic, leaveTangent=0.3036, tangentMode=RCTM_User, time=1, value=0}
        ]}}})
    Module.Index0Float0Write=true
    Module.Index0Float1Write=true
    Module.Index0Float2Write=true
    Module.Index0Float3Write=true
   )
  ]
  rendererClasses [&NiagaraSpriteRendererProperties, &NiagaraMeshRendererProperties]
  renderers [
   {rendererClass=&NiagaraSpriteRendererProperties}
   {rendererClass=&NiagaraMeshRendererProperties, rendererIndex=1}
  ]
 }}
summary {
 systemName=NS_Emote_FloralMane_Spinning
 emitters [{emitterName=Splashes, rendererClasses [&NiagaraSpriteRendererProperties, &NiagaraMeshRendererProperties]}]
 userVariables [User.Direction {type=NiagaraFloat, defaultValue=0, description=""}]
}
