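# Niagara splash effect for the FloralMane emote, spawned in bursts at a skeletal mesh location and tinted pale blue by time of day.
# One condensed emitter fires four spawn groups, using a visibility tag and per-group switches for color, lifetime, shape, motion and fade.
# Three groups fling particles out and up under drag and curve-driven gravity, while group 1 stays still.
# Three SubUV sprite renderers and a mesh renderer draw them, with scale and material parameters animated over life.
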NiagaraSystem /BRCosmetics/Animation/Game/MainPlayer/Emotes/FloralMane/VFX/NS_Emote_FloralMane_Splashes
dependencies {
 usedDataInterfaces [
  &NiagaraDataInterfaceSpriteRendererInfo
  &NiagaraDataInterfaceCurve
  &NiagaraDataInterfaceVector2DCurve
  &NiagaraDataInterfaceSkeletalMesh
 ]
 usedDynamicInputs []
 usedModules [
  &SystemState
  &self:SystemSpawnScript.SystemScriptSource.SystemScriptGraph.NiagaraNodeAssignment_0.SetVariables_#1
 ]
}
emitters {Splashes_Condensed {
  emitterUpdateScript [
   &EmitterState(Life Cycle Mode=::NewEnumerator1, Loop Behavior=::NewEnumerator1, Loop Duration=1, UseLoopDelay=false)
   &SpawnBurst_Instantaneous(Spawn Count=30, Spawn Time=0.1, Use Spawn Probability=false, UseLoopCountLimit=false)
   &SpawnBurst_Instantaneous(Spawn Group=1, Use Spawn Probability=false, UseLoopCountLimit=false) {module=SpawnBurst_Instantaneous001}
   &SpawnBurst_Instantaneous(Spawn Count=30, Spawn Group=2, Use Spawn Probability=false, UseLoopCountLimit=false) {module=SpawnBurst_Instantaneous002}
   &SpawnBurst_Instantaneous(Spawn Count=30, Spawn Group=3, Use Spawn Probability=false, UseLoopCountLimit=false) {module=SpawnBurst_Instantaneous003}
  ]
  particleSpawnScript [
   /SetVariables(Particles.VisibilityTag=&SpawnGroup()) {isSetParameters=true}
   &InitializeParticle(
    Color=&MergeHelper_ColorSwitch(
     0=&FNTimeOfDayColor(
      Night Color=&RandomRangeLinearColor(Minimum=<rgba 0.2555 0.4754 0.8177 0.7043>, Maximum=<rgba 0.2897 0.6737 0.9427 0.8696>)
      Morning Color=&RandomRangeLinearColor(Minimum=<rgba 0.6493 0.7362 0.8177 1>, Maximum=<rgba 0.4664 0.6969 0.9427 1>)
      Day Color=&RandomRangeLinearColor(Minimum=<rgba 0.6392 0.8051 0.974 0.7739>, Maximum=<rgba 0.4484 0.6552 0.9427 0.687>)
      Evening Color=&RandomRangeLinearColor(Minimum=<rgba 0.6493 0.7362 0.8177 1>, Maximum=<rgba 0.4664 0.6969 0.9427 1>)
     )
     1=<rgba 0.4601 0.6396 0.8333 1.3>
     2=&FNTimeOfDayColor(
      Night Color=&RandomRangeLinearColor(Minimum=<rgba 0.2555 0.4754 0.8177 0.7043>, Maximum=<rgba 0.2897 0.6737 0.9427 0.8696>)
      Morning Color=&RandomRangeLinearColor(Minimum=<rgba 0.6493 0.7362 0.8177 1>, Maximum=<rgba 0.4664 0.6969 0.9427 1>)
      Day Color=&RandomRangeLinearColor(Minimum=<rgba 0.6768 0.8377 0.9844 1>, Maximum=<rgba 0.5625 0.7761 1 1>)
      Evening Color=&RandomRangeLinearColor(Minimum=<rgba 0.6493 0.7362 0.8177 1>, Maximum=<rgba 0.4664 0.6969 0.9427 1>)
     )
     3=&FNTimeOfDayColor(
      Night Color=&RandomRangeLinearColor(Minimum=<rgba 0.2555 0.4754 0.8177 0.7043>, Maximum=<rgba 0.2897 0.6737 0.9427 0.8696>)
      Morning Color=&RandomRangeLinearColor(Minimum=<rgba 0.6493 0.7362 0.8177 1>, Maximum=<rgba 0.4664 0.6969 0.9427 1>)
      Day Color=&RandomRangeLinearColor(Minimum=<rgba 0.8097 0.882 0.9479 1>, Maximum=<rgba 0.5625 0.7761 1 1>)
      Evening Color=&RandomRangeLinearColor(Minimum=<rgba 0.6493 0.7362 0.8177 1>, Maximum=<rgba 0.4664 0.6969 0.9427 1>)
     )
     Index=&SpawnGroup()
     Num=::NewEnumerator2
    )
    Lifetime=&MergeHelper_FloatSwitch(
     0=&RandomRangeFloat(Minimum=0.2, Maximum=0.3)
     1=0.45
     2=&RandomRangeFloat(Minimum=0.2, Maximum=0.4)
     3=&RandomRangeFloat(Minimum=0.2, Maximum=0.4)
     Index=&SpawnGroup()
     Num=::NewEnumerator2
    )
    Mass Mode=::NewEnumerator3
    Mesh Scale Mode=::NewEnumerator3
    Mesh Uniform Scale=0.04
    Sprite Rotation Mode=::NewEnumerator2
    Sprite Size Mode=::NewEnumerator1
    Uniform Sprite Size Max=50
    Uniform Sprite Size Min=20
    UsePositionOffset=false
    Write Material Random=true
   )
   &SkeletalMeshLocation(
    MaskSpawnGroup=false
    Mesh Sampling Type=::NewEnumerator5
    Orientation Sampling=::NewEnumerator1
    Position Sampling=::NewEnumerator1
    Sampled Orientation Coordinate Space=::Local
    Sampled Position Coordinate Space=::Local
    Sampled Position Offset=&MergeHelper_VectorSwitch(0=<0 0 -10>, 1=<0 0 4>, 2=<0 0 -10>, 3=<0 0 -10>, Index=&SpawnGroup(), Num=::NewEnumerator2)
    Sampled Velocity Coordinate Space:ENiagaraCoordinateSpace=::Local
    Skeletal Mesh=$System.SKM
    Socket Sampling Mode=::NewEnumerator5
   ) {module=SkeletalMeshLocation001}
   &SubUVAnimation(
    End Frame Range Override=3
    Evaluation Type=::NewEnumerator0
    Output Alternate Renderer Binding=true
    "Sprite Renderer "=DataInterface() {propertyValues {SpriteRenderer=&self:Splashes_Condensed_1.NiagaraSpriteRendererProperties_1}}
    SubUV Animation Mode:ENiagara_SubUVLookupModeV2=::NewEnumerator2
    UseEndFrame=true
    UseStartFrame=true
   )
   &SubUVAnimation(
    End Frame Range Override=3
    Evaluation Type=::NewEnumerator0
    Output Alternate Renderer Binding=true
    "Sprite Renderer "=DataInterface() {propertyValues {SpriteRenderer=&self:Splashes_Condensed_1.NiagaraSpriteRendererProperties_2}}
    SubUV Animation Mode:ENiagara_SubUVLookupModeV2=::NewEnumerator2
    UseEndFrame=true
    UseStartFrame=true
   ) {module=SubUVAnimation001}
   &SubUVAnimation(
    End Frame Range Override=3
    Evaluation Type=::NewEnumerator0
    Output Alternate Renderer Binding=true
    "Sprite Renderer "=DataInterface() {propertyValues {SpriteRenderer=&self:Splashes_Condensed_1.NiagaraSpriteRendererProperties_3}}
    SubUV Animation Mode:ENiagara_SubUVLookupModeV2=::NewEnumerator2
    UseEndFrame=true
    UseStartFrame=true
   ) {module=SubUVAnimation002}
   &InitialMeshOrientation(Mesh Axis To Orient=<0 1 0>, Mesh Orientation Mode=::NewEnumerator2, Orientation Vector=<0 0 1>, Use Rotation Vector=false)
   &ShapeLocation(
    Additional Quaternion Rotation=false
    MaskSpawnGroup=false
    Non Uniform Scale=&MergeHelper_VectorSwitch(0=<0.75 1 1>, 1=<0 0 0>, 2=<1 3.5 1>, 3=<1 3.5 1>, Index=&SpawnGroup(), Num=::NewEnumerator2)
    Override Seed=false
    Ring Radius=&MergeHelper_FloatSwitch(0=45, 1=0, 2=15, 3=15, Index=&SpawnGroup(), Num=::NewEnumerator2)
    Rotation Mode=::NewEnumerator3
    Rotation Quaternion=$Particles.SkeletalMeshLocation001.SampledRotation
    Shape Origin=$Particles.SkeletalMeshLocation001.SampledPosition
    Shape Primitive=::NewEnumerator4
    UseOwnerScale=true
   ) {module=ShapeLocation001}
   &AddVelocity(
    Module.Use Falloff Distance=false
    Origin Offset=<0 0 -2>
    Override Seed=false
    Velocity Mode=::NewEnumerator1
    Velocity Speed=&MergeHelper_FloatSwitch(0=200, 1=0, 2=170, 3=150, Index=&SpawnGroup(), Num=::NewEnumerator2)
   )
   &AddVelocity(Velocity=&MergeHelper_VectorSwitch(
     0=&RandomRangeVector(Minimum=<0 0 100>, Maximum=<0 0 500>)
     1=<0 0 0>
     2=&RandomRangeVector(Minimum=<0 0 100>, Maximum=<0 0 300>)
     3=&RandomRangeVector(Minimum=<0 0 200>, Maximum=<0 0 600>)
     Index=&SpawnGroup()
     Num=::NewEnumerator2
    )) {module=AddVelocity001}
  ]
  particleUpdateScript [
   &ParticleState()
   &ScaleSpriteSize(
    Non-Uniform Curve Sprite Scale=DataInterface() {propertyValues {
      XCurve {keys [
        {arriveTangent=3.4688, interpMode=RCIM_Cubic, leaveTangent=3.4688, tangentMode=RCTM_User, time=-0.0082, value=0.3989}
        {interpMode=RCIM_Cubic, tangentMode=RCTM_User, time=1, value=1}
       ]}
      YCurve {keys [
        {arriveTangent=5.7272, interpMode=RCIM_Cubic, leaveTangent=5.7272, tangentMode=RCTM_User, time=-0.0082, value=0.3989}
        {interpMode=RCIM_Cubic, tangentMode=RCTM_User, time=1, value=1}
       ]}
     }}
    Scale Sprite Size Mode=::NewEnumerator3
   )
   &ScaleMeshSize(Scale Factor=&VectorFromFloat(Value=&FloatFromCurve(
      FloatCurve=DataInterface() {propertyValues {Curve {
         defaultValue=1
         keys [
          {arriveTangent=2.9253, interpMode=RCIM_Cubic, leaveTangent=2.9253, tangentMode=RCTM_User, time=0.003, value=0.3971}
          {arriveTangent=0.8054, interpMode=RCIM_Cubic, leaveTangent=0.8054, time=0.3598, value=1.0164}
          {arriveTangentWeight=0.7014, interpMode=RCIM_Cubic, tangentMode=RCTM_User, time=1, value=1.2}
         ]
        }}}
      Scale Curve=2
     )))
   &FN_Radial_Material(
    Base Sprite Tile U=3
    Depth Fade=1
    Radial Mask Power=3
    Sprite Extrusion Speed:NiagaraFloat=&FloatFromCurve(FloatCurve=DataInterface() {propertyValues {Curve {keys [
         {interpMode=RCIM_Cubic, tangentMode=RCTM_User, time=0, value=1.3}
         {arriveTangent=-0.9, interpMode=RCIM_Cubic, leaveTangent=-0.9, time=0.1592, value=0.7203}
         {interpMode=RCIM_Cubic, tangentMode=RCTM_User, time=1, value=0.4}
        ]}}})
    Sprite Spin Speed:NiagaraFloat=&FloatFromCurve(FloatCurve=DataInterface() {propertyValues {Curve {keys [<at 0 -1>, {arriveTangent=1.5942, leaveTangent=1.5942, time=1, value=-0.2}]}}})
    Write to Module Namespace=true
   )
   &Drag(
    Drag=&MergeHelper_FloatSwitch(
     0=&RandomRangeFloat(Minimum=0.3, Maximum=1.3)
     1=0
     2=&RandomRangeFloat(Minimum=0.3, Maximum=1.3)
     3=&RandomRangeFloat(Minimum=0.5, Maximum=1.3)
     Index=&SpawnGroup()
     Num=::NewEnumerator2
    )
    Use Linear Drag=true
    Use Rotational Drag=true
   )
   &GravityForce(Gravity=&MakeVector(
     X=0
     Y=0
     Z=&MergeHelper_FloatSwitch(
      0=&FloatFromCurve(FloatCurve=DataInterface() {propertyValues {Curve {keys [<at 0.012 -821.1881>,<at 0.5013 -1845.0896>]}}}, Scale Curve=1.5)
      1=0
      2=&FloatFromCurve(FloatCurve=DataInterface() {propertyValues {Curve {keys [<at 0.012 -821.1881>,<at 0.5981 -1809.3721>]}}}, Scale Curve=1.5)
      3=&FloatFromCurve(FloatCurve=DataInterface() {propertyValues {Curve {keys [<at 0.012 -821.1881>,<at 0.5997 -1821.2778>]}}}, Scale Curve=1.7)
      Index=&SpawnGroup()
      Num=::NewEnumerator2
     )
    ))
   &SolveForcesAndVelocity(Clamp Velocity=false, Limit Acceleration=false)
   &ScaleColor(
    Scale Alpha=&MergeHelper_FloatSwitch(
     0=&FloatFromCurve(
      FloatCurve=DataInterface() {propertyValues {Curve {
         defaultValue=1
         keys [
          {arriveTangent=22.5986, interpMode=RCIM_Cubic, leaveTangent=22.5986, tangentMode=RCTM_User, time=0.1264, value=0}
          {interpMode=RCIM_Cubic, tangentMode=RCTM_User, time=0.283, value=1.0072}
          {arriveTangent=-0.0835, tangentMode=RCTM_User, time=0.9969, value=-0.0194}
         ]
        }}}
      Scale Curve=&RandomRangeFloat(Minimum=0.6, Maximum=0.85)
     )
     1=&FloatFromCurve(FloatCurve=DataInterface() {propertyValues {Curve {keys [
          <at -0.0046 1.0113 RCIM_Cubic>
          {arriveTangent=-1.3263, interpMode=RCIM_Cubic, leaveTangent=-1.3263, time=0.3771, value=0.2858}
          <at 0.7564 0.002 RCIM_Cubic>
         ]}}})
     2=&FloatFromCurve(
      FloatCurve=DataInterface() {propertyValues {Curve {
         defaultValue=1
         keys [
          {arriveTangent=14.3798, interpMode=RCIM_Cubic, leaveTangent=14.3798, tangentMode=RCTM_User, time=0.1264, value=0}
          {interpMode=RCIM_Cubic, tangentMode=RCTM_User, time=0.2984, value=0.995}
          {arriveTangent=-0.0835, tangentMode=RCTM_User, time=0.9969, value=-0.0194}
         ]
        }}}
      Scale Curve=&RandomRangeFloat(Minimum=0.6, Maximum=0.85)
     )
     3=&FloatFromCurve(
      FloatCurve=DataInterface() {propertyValues {Curve {
         defaultValue=1
         keys [
          {arriveTangent=22.7037, interpMode=RCIM_Cubic, leaveTangent=22.7037, tangentMode=RCTM_User, time=0.0022, value=0}
          {interpMode=RCIM_Cubic, tangentMode=RCTM_User, time=0.1904, value=0.9461}
          {arriveTangent=-0.0835, tangentMode=RCTM_User, time=0.9969, value=-0.0194}
         ]
        }}}
      Scale Curve=&RandomRangeFloat(Minimum=0.6, Maximum=0.85)
     )
     Index=&SpawnGroup()
     Num=::NewEnumerator2
    )
    ScaleA=true
    ScaleRGB=true
   ) {module=ScaleColor001}
   /SetVariables(Particles.Splashes Dynamic Param:Vector4f=&MakeVector4(
     X=&FloatFromCurve(FloatCurve=DataInterface() {propertyValues {Curve {keys [
          <at 0.1543 1 RCIM_Cubic>
          {arriveTangent=-2.4268, interpMode=RCIM_Cubic, leaveTangent=-2.4268, tangentMode=RCTM_User, time=0.798, value=0.3996}
          {arriveTangent=0.3036, interpMode=RCIM_Cubic, leaveTangent=0.3036, tangentMode=RCTM_User, time=1, value=0}
         ]}}})
     Y=0
     Z=0
     W=0
    )) {isSetParameters=true}
  ]
  rendererClasses [&NiagaraSpriteRendererProperties, &NiagaraMeshRendererProperties]
  renderers [
   {rendererClass=&NiagaraSpriteRendererProperties}
   {rendererClass=&NiagaraMeshRendererProperties, rendererIndex=1}
   {rendererClass=&NiagaraSpriteRendererProperties, rendererIndex=2}
   {rendererClass=&NiagaraSpriteRendererProperties, rendererIndex=3}
  ]
 }}
summary {
 systemName=NS_Emote_FloralMane_Splashes
 emitters [{
   emitterName=Splashes_Condensed
   rendererClasses [&NiagaraSpriteRendererProperties, &NiagaraMeshRendererProperties]
  }]
 userVariables [
  User.bIsFrontEnd {type=NiagaraBool}
  User.bIsFrontEndPreview {type=NiagaraBool}
  User.BurstIndex {type=NiagaraFloat, defaultValue=0, description=""}
  User.OwnerLocalVelocity {type=Vector3f, description="", defaultValue {struct=Vector3f, value=<0 0 0>}}
 ]
}
